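# Translucent unlit flame-edge material for Niagara mesh particles on the spectral Gatehound cosmetic.
# Panning distorted noise, UV edge masks and caustics shape the opacity and emissive, tinted by a two-color vertical gradient and particle color.
# World position offset adds a normal-based wobble to a clamped, UV-weighted bend driven by Niagara dynamic parameters relative to particle position.
# Depth fade softens intersections, except at Low and Medium quality.
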Material /BRCosmetics/Characters/Player/Male/Medium/Bodies/M_MED_GateHound_Spectral/FX/Materials/M_Gatehound_FlameEdge
graph {
 properties {
  bUsedWithNiagaraMeshParticles=true
  blendMode=BLEND_Translucent
  pixelDepthOffsetMode=PDOM_Legacy
  shadingModel=MSM_Unlit
  usedShadingModels=Unlit
 }
 nodes [
  n1 = DynamicParameter() {desc=Parameters set by Niagara, parameterIndex=1, paramNames [X,Y,Z,WPOIntensity]}
  n2 = Clamp(
   None=Subtract(
    A=MaterialFunctionCall(X=@n1.X, Y=@n1.X, Z=@n1.X) {materialFunction=&MakeFloat3}.Result
    B=ParticlePositionWS()
   )
   Min=ScalarParameter() {defaultValue=-100, group=Interpolate WPO, parameterName=Clamp_min}
   Max=ScalarParameter() {defaultValue=100, group=Interpolate WPO, parameterName=Clamp_max}
  ) {maxDefault=1000, minDefault=-1000}
  n3 = OneMinus(None=TextureCoordinate())
  n4 = Multiply(
   A=Power(Base=ComponentMask(None=@n3) {g=true}, Exp=ScalarParameter() {defaultValue=4, group=Interpolate WPO, parameterName=WPOBendPower})
   B=@n1.WPOIntensity
  )
  n5 = Panner(Coordinate=TextureCoordinate(), Time=Multiply(A=Time(), B=ScalarParameter() {defaultValue=1.1, group=Panner Speed, parameterName=Speed})) {speedY=0.8}
  n6 = ComponentMask(None=TextureSample(UVs=@n5) {samplerType=SAMPLERTYPE_Normal, texture=&T_Weap_Distort_N_02}.RGB) {g=true, r=true}
  n7 = TextureSample(UVs=@n5) {texture=&T_Smooth_tile}
  n8 = ComponentMask(None=TextureCoordinate()) {g=true}
  n9 = LinearInterpolate(B=Multiply(A=VertexNormalWS(), B=AppendVector(A=@n6, B=@n7.R)), Alpha=Power(Base=OneMinus(None=@n8)) {constExponent=1.5})
  n10 = ScalarParameter() {defaultValue=6, group=WPO, parameterName=WaveIntensity, sortPriority=7}
  Add(
   A=LinearInterpolate(A=Multiply(A=@n9, B=@n10) {constB=10}, Alpha=ScalarParameter() {desc=no_Wpo, group=Simple_Round_particle, parameterName=No_WPO})
   B=LinearInterpolate(
    A=Constant()
    B=LinearInterpolate(B=@n2, Alpha=@n4)
    Alpha=ScalarParameter() {defaultValue=1, group=Interpolate WPO, parameterName=Use interpolate WPO}
   )
  )
  n11 = ComponentMask(None=TextureCoordinate()) {g=true}
  n12 = LinearInterpolate(
   A=VectorParameter() {parameterName=ColorA, defaultValue=<rgba 0 0 0 1>}.RGB
   B=VectorParameter() {parameterName=ColorB, defaultValue=<rgba 0 0 0 1>}.RGB
   Alpha=SmoothStep(
    Min=ScalarParameter() {defaultValue=0.2, parameterName=ColoGradientMin}
    Max=ScalarParameter() {defaultValue=0.75, parameterName=ColoGradientMax}
    Value=OneMinus(None=@n11)
   ) {constMax=0.75, constMin=0.33}
  )
  n13 = MaterialFunctionCall(Float2=TextureCoordinate()) {materialFunction=&BreakOutFloat2Components}
  n14 = TextureCoordinate()
  n15 = VectorParameter() {parameterName=UV_Size, defaultValue=<rgba 3 1 0.5 0.5>}
  n16 = AppendVector(A=@n15.R, B=@n15.R)
  n17 = Add(
   A=LinearInterpolate(B=LinearInterpolate(A=@n6, Alpha=@n13.G) {constB=0}, Alpha=Constant() {r=0.5})
   B=Panner(Coordinate=Multiply(A=@n14, B=@n16)) {speedY=1}
  )
  n18 = TextureSampleParameter2D(UVs=@n17) {
   group=Fire Shape
   parameterName=Fire Micro Texture
   samplerType=SAMPLERTYPE_Masks
   sortPriority=6
   texture=/Game/Global/Textures/Default/Noise/T_WrapEdgeWear_Noise
  }
  n19 = Power(
   Base=Sine(None=Multiply(A=@n13.R) {constB=0.5})
   Exp=LinearInterpolate(
    A=ScalarParameter() {defaultValue=4, group=Fire Shape, parameterName=Edge Fade Value Top, sortPriority=3}
    B=ScalarParameter() {defaultValue=0.5, group=Fire Shape, parameterName=Edge Fade Value Bottom, sortPriority=4}
    Alpha=@n13.G
   ) {constA=6, constB=2}
  ) {constExponent=4}
  n20 = SmoothStep(Value=SphereMask(A=TextureCoordinate(), B=Constant2Vector() {g=-0.45, r=0.5}) {attenuationRadius=1.4, hardnessPercent=0.25}) {constMax=0.25}
  n21 = Multiply(A=@n13.G, B=Multiply(A=Multiply(A=@n19, B=@n20)))
  n22 = Multiply(A=@n18.R, B=@n21)
  n23 = Panner(Coordinate=Multiply(A=@n14, B=AppendVector(A=@n15.B, B=@n15.B))) {speedY=0.5}
  n24 = Power(Base=OneMinus(None=@n13.G))
  n25 = SmoothStep(
   Min=ScalarParameter() {defaultValue=0.4, parameterName=NoiseGradMin}
   Max=ScalarParameter() {defaultValue=1, parameterName=NoiseGradMax}
   Value=Power(
    Base=TextureSample(UVs=@n23) {texture=/Game/Global/Textures/Default/Caustics/T_Caustics_Softer}.R
    Exp=LinearInterpolate(Alpha=@n24) {constB=0.5}
   )
  ) {constMin=0.4}
  n26 = Multiply(A=Power(Base=Multiply(A=@n22) {constB=2.5}), B=@n25)
  n27 = Add(A=Multiply(A=SmoothStep(Value=@n11) {constMin=0.25}, B=Saturate(None=@n26))) {constB=0.5}
  MaterialFunctionCall(Emissive=Multiply(A=Multiply(A=@n12, B=@n27), B=ParticleColor().RGB)) {materialFunction=&MF_FX_EmissiveAdjustments}
  n28 = SmoothStep(Value=LinearInterpolate(A=@n22, Alpha=@n21)) {constMax=0.7}
  n29 = TextureSample(UVs=@n17) {texture=/BRCosmetics/Athena/Effects/Particles/Skydiving/NS_Contrail_Chainmail/T_WispyFireNoise03}
  n30 = SmoothStep(Value=OneMinus(None=LinearInterpolate(A=@n29.R, Alpha=@n21) {constB=0}))
  n31 = Multiply(A=Power(Base=Multiply(A=@n28, B=@n30)), B=ParticleColor().A)
  n32 = LinearInterpolate(B=@n25, Alpha=ScalarParameter() {parameterName=NoiseStrenght}) {constA=1}
  n33 = DepthFade(FadeDistance=ScalarParameter() {defaultValue=5, parameterName=FadeDistance}) {fadeDistanceDefault=5}
  n34 = Constant() {r=1}
  Multiply(A=Saturate(None=Multiply(A=@n31, B=@n32)), B=QualitySwitch(Default=@n33, Low=@n34, High=@n33, Medium=@n34, Epic=@n33))
 ]
}
