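# Unlit translucent, two-sided Niagara mesh-particle material that draws spline flame ribbons for the spectral GateHound body's FX.
# World position offset combines SplineThicken width, a masked sine wobble, and a U-weighted bend by a Niagara-set position.
# Opacity erodes a panning wispy mask along U, modulated by caustics noise, with depth fade outside Low/Medium quality.
# Emissive blends two colors along U, scaled by the mask and particle color.
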Material /BRCosmetics/Characters/Player/Male/Medium/Bodies/M_MED_GateHound_Spectral/FX/Materials/M_Gatehound_Flame_Spline
graph {
 properties {
  bUsedWithNiagaraMeshParticles=true
  blendMode=BLEND_Translucent
  pixelDepthOffsetMode=PDOM_Legacy
  shadingModel=MSM_Unlit
  twoSided=true
  usedShadingModels=Unlit
 }
 nodes [
  n1 = DynamicParameter() {desc=Parameters set by Niagara, parameterIndex=1, paramNames [X,Y,Z,WPOIntensity]}
  n2 = MaterialFunctionCall(X=@n1.X, Y=@n1.X, Z=@n1.X) {materialFunction=&MakeFloat3}
  n3 = ParticlePositionWS()
  n4 = VertexColor()
  n5 = Clamp(
   None=Multiply(A=Subtract(A=@n2.Result, B=@n3), B=Add(A=@n4.G) {constB=0.5})
   Min=ScalarParameter() {defaultValue=-100, group=Interpolate WPO, parameterName=Clamp_min}
   Max=ScalarParameter() {defaultValue=100, group=Interpolate WPO, parameterName=Clamp_max}
  ) {maxDefault=1000, minDefault=-1000}
  n6 = TextureCoordinate()
  n7 = Multiply(
   A=Power(Base=ComponentMask(None=@n6) {r=true}, Exp=ScalarParameter() {defaultValue=4, group=Interpolate WPO, parameterName=WPOBendPower})
   B=@n1.WPOIntensity
  )
  n8 = DynamicParameter() {defaultValue=<rgba 1 1 1 0>, paramNames [WPO, Speed, WPO Disturbance, U offset]}
  n9 = Multiply(A=Multiply(A=ScalarParameter() {defaultValue=15, parameterName=WidthBase}, B=@n8.WPO))
  n10 = Multiply(A=ScalarParameter() {defaultValue=25, parameterName=WidthTip}, B=@n8.WPO)
  n11 = StaticBool()
  n12 = DynamicParameter() {defaultValue=<rgba 1 1 1 0>, paramNames [WPO, Speed, WPO Disturbance, U offset]}
  n13 = Constant()
  n14 = ComponentMask(None=Add(A=TextureCoordinate() {uTiling=2.2}, B=AppendVector(A=@n12.U offset, B=@n13))) {r=true}
  n15 = Multiply(A=Add(A=Time(), B=VertexColor().R)) {constB=-1.3}
  n16 = Multiply(A=Sine(None=Add(A=@n14, B=@n15)) {period=1.2}, B=VectorParameter() {parameterName=WPO Direction, defaultValue=<rgba 0 0 1 1>}.RGB)
  n17 = Multiply(
   A=LinearInterpolate(
    A=Transform(None=@n16) {transformSourceType=TRANSFORMSOURCE_World, transformType=TRANSFORM_Camera}
    B=Transform(None=@n16) {transformSourceType=TRANSFORMSOURCE_Local}
    Alpha=ScalarParameter() {parameterName=Change Transform}
   )
   B=DynamicParameter() {defaultValue=<rgba 1 1 1 0>, paramNames [WPO, Speed, WPO Disturbance, U offset]}.WPO Disturbance
  )
  n18 = TextureCoordinate()
  n19 = Saturate(None=SmoothStep(Value=ComponentMask(None=@n18) {r=true}) {constMin=0.1})
  Add(
   A=Add(
    A=MaterialFunctionCall(WidthBase=@n9, WidthTip=@n10, Expand U or V UV Channel=@n11) {materialFunction=&SplineThicken}.WorldPositionOffset
    B=Multiply(A=@n17, B=@n19)
   )
   B=LinearInterpolate(
    A=Constant()
    B=LinearInterpolate(B=@n5, Alpha=@n7)
    Alpha=ScalarParameter() {defaultValue=1, group=Interpolate WPO, parameterName=Use interpolate WPO}
   )
  )
  n20 = TextureCoordinate()
  n21 = VectorParameter() {parameterName=FlameUVScale, defaultValue=<rgba 1 1 0.25 0.25>}
  n22 = Multiply(A=@n20, B=AppendVector(A=@n21.R, B=@n21.R))
  n23 = Add(A=Time(), B=VertexColor().R)
  n24 = TextureSample(UVs=MaterialFunctionCall(UVs=Panner(Coordinate=@n22, Time=@n23) {speedX=-1}, Texture Scale=Constant2Vector() {g=0.8, r=1}) {materialFunction=&ScaleUVsByCenter}.UVs) {
   samplerSource=SSM_Wrap_WorldGroupSettings
   texture=/BRCosmetics/Athena/Effects/Particles/Skydiving/NS_Contrail_Chainmail/T_Chainmail_WispyMaskFat
  }
  n25 = ComponentMask(None=TextureCoordinate()) {r=true}
  n26 = Subtract(A=@n24.R, B=@n25)
  n27 = SphereMask(A=TextureCoordinate() {uTiling=0}, B=Constant2Vector() {g=0.5}) {attenuationRadius=0.5, hardnessPercent=0}
  n28 = SmoothStep(Value=@n25) {constMax=0.25}
  n29 = Multiply(A=Multiply(A=@n26, B=@n27), B=@n28)
  n30 = Multiply(A=@n28, B=Multiply(A=OneMinus(None=@n25), B=@n27))
  n31 = ScalarParameter() {parameterName=Textureblend}
  n32 = ScalarParameter() {defaultValue=0.4, parameterName=NoiseGradMin}
  n33 = ScalarParameter() {defaultValue=1, parameterName=NoiseGradMax}
  n34 = Multiply(A=@n20, B=AppendVector(A=@n21.B, B=@n21.B))
  n35 = Power(
   Base=TextureSample(UVs=Panner(Coordinate=@n34, Time=@n23) {speedX=-0.25}) {texture=/Game/Global/Textures/Default/Caustics/T_Caustics_Softer}.R
   Exp=LinearInterpolate(Alpha=Power(Base=@n25)) {constB=0.5}
  )
  n36 = Multiply(
   A=SmoothStep(Value=LinearInterpolate(A=@n29, B=@n30, Alpha=@n31)) {constMax=0.5, constValue=1}
   B=LinearInterpolate(B=SmoothStep(Min=@n32, Max=@n33, Value=@n35) {constMin=0.4}, Alpha=ScalarParameter() {defaultValue=1, parameterName=NoiseStrenght}) {constA=1}
  )
  n37 = ParticleColor()
  n38 = DepthFade(FadeDistance=ScalarParameter() {defaultValue=7, parameterName=FadeDistance}) {fadeDistanceDefault=7}
  n39 = Constant() {r=1}
  Multiply(A=Multiply(A=Saturate(None=@n36), B=@n37.A), B=QualitySwitch(Default=@n38, Low=@n39, High=@n38, Medium=@n39, Epic=@n38))
  n40 = ScalarParameter() {defaultValue=0.184, parameterName=ColoGradientMin}
  n41 = ScalarParameter() {defaultValue=0.788, parameterName=ColoGradientMax}
  n42 = ComponentMask(None=TextureCoordinate()) {r=true}
  n43 = Multiply(
   A=LinearInterpolate(
    A=VectorParameter() {parameterName=ColorA, defaultValue=<rgba 0 0 0 1>}.RGB
    B=VectorParameter() {parameterName=ColorB, defaultValue=<rgba 0 0 0 1>}.RGB
    Alpha=SmoothStep(Min=@n40, Max=@n41, Value=@n42) {constMax=0.75, constMin=0.33}
   )
   B=Add(A=Saturate(None=@n26)) {constB=0.75}
  )
  Reroute(None=MaterialFunctionCall(Emissive=Multiply(A=@n43, B=@n37.RGB)) {materialFunction=&MF_FX_EmissiveAdjustments}.Result)
 ]
}
