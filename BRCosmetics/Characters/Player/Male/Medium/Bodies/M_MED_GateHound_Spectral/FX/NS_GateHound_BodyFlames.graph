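# Niagara system that dresses the spectral GateHound player body in persistent flames at neck, shoulders, feet and apparently hands, plus foot glow sprites and GPU sparks.
# Burst-spawned mesh flames follow skeletal bones or sockets each update, computing smoothed positions alongside animated material parameters, likely for trailing motion.
# VariantNumber selects colors from user arrays, dimmed by time of day.
# Visibility flags hide foot and hand effects; some flames also scale by distance.
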NiagaraSystem /BRCosmetics/Characters/Player/Male/Medium/Bodies/M_MED_GateHound_Spectral/FX/NS_GateHound_BodyFlames
dependencies {
 usedDataInterfaces [
  &NiagaraDataInterfaceArrayFloat
  &NiagaraDataInterfaceArrayFloat3
  &NiagaraDataInterfaceArrayInt32
  &NiagaraDataInterfaceCurve
  &NiagaraDataInterfaceSkeletalMesh
 ]
 usedDynamicInputs [&SetBoolByFloatComparison, &VectorLength]
 usedModules [
  &self:SystemSpawnScript.SystemScriptSource.SystemScriptGraph.NiagaraNodeAssignment_0.SetVariables_#1
  &self:SystemSpawnScript.SystemScriptSource.SystemScriptGraph.NiagaraNodeAssignment_1.SetVariables_#2
  &SystemState
  &self:SystemSpawnScript.SystemScriptSource.SystemScriptGraph.NiagaraNodeAssignment_2.SetVariables_#3
  &self:SystemSpawnScript.SystemScriptSource.SystemScriptGraph.NiagaraNodeAssignment_3.SetVariables_#4
  &self:SystemSpawnScript.SystemScriptSource.SystemScriptGraph.NiagaraNodeAssignment_4.SetVariables_#5
 ]
}
emitters {
 Fire_Spline_Feet {
  emitterUpdateScript [
   &EmitterState(
    Inactive Response=::NewEnumerator1
    Life Cycle Mode=::NewEnumerator1
    Loop Behavior=::NewEnumerator1
    Loop Duration Mode=::NewEnumerator1
    UseLoopDelay=false
   )
   &SpawnBurst_Instantaneous(Spawn Count=2, Use Spawn Probability=false, UseLoopCountLimit=false)
  ]
  particleSpawnScript [
   &PartitionParticles(Number Of Partitions=4)
   &InitializeParticle(
    Mesh Scale=&SelectVectorFromArray(
     Direct Array Index=$Particles.PartitionParticles.Partition
     Vector Selection Array=DataInterface() {propertyValues {FloatData [<0.5 0.5 0.8>, <0.5 0.5 0.8>]}}
    )
    Mesh Scale Mode=::NewEnumerator4
    Sprite Size Mode=::NewEnumerator3
    Uniform Sprite Size=25
    UsePositionOffset=false
    Write Material Random=true
   )
   &InitialMeshOrientation(
    Mesh Axis To Orient=&SelectVectorFromArray(
     Direct Array Index=$Particles.PartitionParticles.Partition
     Vector Selection Array=DataInterface() {propertyValues {FloatData [<1 0 0>, <0 0 -1>]}}
    )
    Mesh Orientation Mode=::NewEnumerator2
    Orientation Vector=&SelectVectorFromArray(
     Direct Array Index=$Particles.PartitionParticles.Partition
     Vector Selection Array=DataInterface() {propertyValues {FloatData [<0 0 -1>, <1 0 0>]}}
    )
    Rotation=&SelectVectorFromArray(
     Direct Array Index=$Particles.PartitionParticles.Partition
     Vector Selection Array=DataInterface() {propertyValues {FloatData [<0.125 0 0>, <-0.125 0 0>]}}
    )
    Use Rotation Vector=true
   )
   &CameraOffset(Camera Offset Amount=-2)
  ]
  particleUpdateScript [
   &SkeletalMeshLocation(
    Bone / Socket Index=$Particles.PartitionParticles.Partition
    Bone Sampling Mode=::NewEnumerator4
    MaskSpawnGroup=false
    Sampled Position Offset=&SelectVectorFromArray(
     Direct Array Index=$Particles.PartitionParticles.Partition
     Vector Selection Array=DataInterface() {propertyValues {FloatData [<-2 -4 2>, <2 4 -2>]}}
    )
    Skeletal Mesh=$System.Skeletal Mesh
   )
   &ParticleState(Kill Particles When Lifetime Has Elapsed=false, Let Infinitely Lived Particles Die When Emitter Deactivates=true)
   &Timeline(Play Rate=2, Timeline End=10000000)
   &ScaleMeshSize(Scale Factor=&Multiply_Vector(
     A=&Multiply_VectorByFloat(Vector=<0.2 0.2 0.2>, Float=0.7)
     B=&Lerp_Vector(
      A=&Lerp_Vector(
       A=&Lerp_Vector(A=<1 1 1>, B=<1 1 1>, Alpha=$System.InterpolateOverTime.InterpolateOverTime_Float.Moving Average)
       B=<0.8 0.8 0.8>
       Alpha=$System.InterpolateOverTime.InterpolateOverTime_Float.Moving Average
      )
      B=<0.8 0.8 0.8>
      Alpha=&ClampFloat(Float=$System.InterpolateOverTime.InterpolateOverTime_Float.Moving Average)
     )
    ))
   &ScaleColor(
    Scale Alpha=0.33
    Scale RGB=&MakeVectorFromLinearColorRGB(Color=&MultiplyLinearColor(
      Linear Color=&SelectLinearColorFromArray(
       Array Sampling Mode=::NewEnumerator1
       Direct Array Index=$User.VariantNumber
       Color Selection Array=$User.VariantColor_B1_ColorArray
       Random Seed=?
      )
      Scale Factor=&MakeVector4_FromLinearColor(LinearColor=&FNTimeOfDayColor(Night Color=<rgba 0.7 0.7 0.7 1>, Morning Color=<rgba 0.9 0.9 0.9 1>, Day Color=<rgba 1 1 1 1>, Evening Color=<rgba 0.9 0.9 0.9 1>))
     ))
    ScaleA=true
    ScaleRGB=true
   )
   &InterpolateOverTime(Rate Of Change=10, Target Vector Value:NiagaraPosition=$Particles.Position, Variable Type=::NewEnumerator2)
   /SetVariables(Particles.Interpolate DY:Vector4f=&MakeFloat4FromVectorAndFloat("Vector (RGB)"=$Particles.InterpolateOverTime.InterpolateOverTime_V3_.Moving Average, "Float (Alpha)"=0.5)) {isSetParameters=true, module=SetVariables_#6}
   /SetVariables(Particles.Color1:LinearColor=&MakeLinearColorFromVectorAndFloat(
     "Vector (RGB)"=&Multiply_Vector(A=&MakeVectorFromLinearColorRGB(Color=$Particles.Color), B=<0.25 0.25 0.25>)
     "Float (Alpha)"=&MakeCustomFloatFromBool(Bool=$User.bisFootHidden, True Float=0, False Float=1)
    )) {isSetParameters=true, module=SetVariables_#7}
   /SetVariables(Particles.Color2=&MakeLinearColorFromVectorAndFloat(
     "Vector (RGB)"=&Multiply_Vector(A=&MakeVectorFromLinearColorRGB(Color=$Particles.Color), B=<0.45 0.45 0.45>)
     "Float (Alpha)"=&MakeCustomFloatFromBool(Bool=$User.bisFootHidden, True Float=0, False Float=1)
    )) {isSetParameters=true, module=SetVariables_#8}
   /SetVariables(Particles.Color3:LinearColor=&MakeLinearColorFromVectorAndFloat(
     "Vector (RGB)"=&Multiply_Vector(A=&MakeVectorFromLinearColorRGB(Color=$Particles.Color), B=<1 1 1>)
     "Float (Alpha)"=&MakeCustomFloatFromBool(Bool=$User.bisFootHidden, True Float=0, False Float=1)
    )) {isSetParameters=true}
   &DynamicMaterialParameters(
    Index 0 Param 1=&Lerp_Float(
     A=&SelectFloatFromArray(Direct Array Index=$Particles.PartitionParticles.Partition, Float Selection Array=DataInterface() {propertyValues {FloatData [0.5,0.5]}})
     B=0.8
     Alpha=0
    )
    Index 0 Param 2=&Multiply_Float(A=$Particles.Timeline.Time, B=1.25)
    Index 0 Param 3=3
    Index 0 Param 4=&SelectFloatFromArray(Direct Array Index=$Particles.PartitionParticles.Partition, Float Selection Array=DataInterface() {propertyValues {FloatData [0,0.33]}})
    Module.Index0Param0Write=true
    Module.Index0Param1Write=true
    Module.Index0Param2Write=true
    Module.Index0Param3Write=true
   )
   /SetVariables(
    Particles.GlowCameraOffset=10
    Particles.GlowColor=&MultiplyLinearColor(
     Linear Color=$Particles.Color
     Scale Factor=&MakeVector4_FromFloat(Float=&Multiply_Float(
       A=0.5
       B=&Waveform(
        "Waveform [1]"=::NewEnumerator12
        "[1] Phase Scale"=5
        "[1] Phase"=&Add_Float(
         A=$Emitter.Age
         B=&SelectFloatFromArray(Direct Array Index=$Particles.PartitionParticles.Partition, Float Selection Array=DataInterface() {propertyValues {FloatData [0,0.22]}})
        )
        "[1] Frequency"=2
        "[1] Amplitude Min/Max"=<0 1>
        "[1] Scale"=0.5
        "[1] Bias"=1
       )
      ))
    )
   ) {isSetParameters=true}
  ]
  rendererClasses [&NiagaraMeshRendererProperties]
  renderers [
   {rendererClass=&NiagaraMeshRendererProperties, rendererIndex=0}
   {rendererClass=&NiagaraMeshRendererProperties, rendererIndex=1}
   {rendererClass=&NiagaraMeshRendererProperties, rendererIndex=2}
   {rendererClass=&NiagaraMeshRendererProperties, rendererIndex=3}
  ]
 }
 Fire_Spline_Feet001 {
  emitterUpdateScript [
   &EmitterState(
    Inactive Response=::NewEnumerator1
    Life Cycle Mode=::NewEnumerator1
    Loop Behavior=::NewEnumerator1
    Loop Duration Mode=::NewEnumerator1
    UseLoopDelay=false
   )
   &SpawnBurst_Instantaneous(Spawn Count=6, Use Spawn Probability=false, UseLoopCountLimit=false)
  ]
  particleSpawnScript [
   &PartitionParticles(Number Of Partitions=6)
   &InitializeParticle(
    Mesh Scale=&SelectVectorFromArray(
     Direct Array Index=$Particles.PartitionParticles.Partition
     Vector Selection Array=DataInterface() {propertyValues {FloatData [<0.5 0.5 0.8>, <0.5 0.5 0.8>, <0.5 0.5 0.8>, <0.5 0.5 0.8>, <0.5 0.5 0.8>, <0.5 0.5 0.8>]}}
    )
    Mesh Scale Mode=::NewEnumerator4
    Sprite Size Mode=::NewEnumerator3
    Uniform Sprite Size=12
    UsePositionOffset=false
    Write Material Random=true
   )
   &InitialMeshOrientation(
    Mesh Orientation Mode=::NewEnumerator6
    Orientation Quaternion:Quat4f=<0 -1 0 1>
    Rotation=&MakeVector(
     X=0
     Y=&SelectFloatFromArray(
      Direct Array Index=$Particles.PartitionParticles.Partition
      Float Selection Array=DataInterface() {propertyValues {FloatData [0.06,0.06,0.02,0.02,-0.02,-0.02]}}
     )
     Z=0
    )
    Use Rotation Vector=true
   )
  ]
  particleUpdateScript [
   &SkeletalMeshLocation(
    Bone / Socket Index=&SelectIntFromArray(Direct Array Index=$Particles.PartitionParticles.Partition, Int Selection Array=DataInterface() {propertyValues {IntData [0,1,2,3,4,5]}})
    MaskSpawnGroup=false
    Mesh Sampling Type=::NewEnumerator5
    Skeletal Mesh=$System.Skeletal Mesh
    Socket Sampling Mode=::NewEnumerator5
   )
   /SetVariables(Particles.ADSFade:NiagaraFloat=&Lerp_Float(
     A=0
     B=1
     Alpha=&FN_LOD_Distance_Remap(Start Distance=250, End Distance=450, Minimum Output=0.5, Maximum Output=1, Clamp Output=true)
    )) {isSetParameters=true, module=SetVariables_#9}
   &ParticleState(Kill Particles When Lifetime Has Elapsed=false, Let Infinitely Lived Particles Die When Emitter Deactivates=true)
   &Timeline(
    Play Rate=&SelectFloatFromArray(
     Direct Array Index=$Particles.PartitionParticles.Partition
     Float Selection Array=DataInterface() {propertyValues {FloatData [2.5,2.5,2.2,2.2,1.75,1.75]}}
    )
    Timeline End=10000000
   )
   &ScaleMeshSize(Scale Factor=&Multiply_Vector(
     A=&Multiply_VectorByFloat(
      Vector=<0.2 0.2 0.2>
      Float=&Multiply_Float(
       A=&SelectFloatFromArray(
        Direct Array Index=$Particles.PartitionParticles.Partition
        Float Selection Array=DataInterface() {propertyValues {FloatData [0.4,0.4,0.5,0.5,0.6,0.6]}}
       )
       B=&Multiply_Float(A=$System.bisHandHidden, B=$System.FlameDistanceScaling)
      )
     )
     B=&Lerp_Vector(
      A=&Lerp_Vector(
       A=&Lerp_Vector(A=<1 1 1>, B=<1 1 1>, Alpha=$System.InterpolateOverTime.InterpolateOverTime_Float.Moving Average)
       B=<0.8 0.8 0.8>
       Alpha=$System.InterpolateOverTime.InterpolateOverTime_Float.Moving Average
      )
      B=<0.5 0.5 0.5>
      Alpha=&ClampFloat(Float=$System.InterpolateOverTime.InterpolateOverTime_Float.Moving Average)
     )
    ))
   &ScaleColor(
    Scale Alpha=&Multiply_Float(A=&Multiply_Float(A=0.5, B=$System.FlameDistanceScaling), B=$Particles.ADSFade)
    Scale RGB=&MakeVectorFromLinearColorRGB(Color=&MultiplyLinearColor(
      Linear Color=&SelectLinearColorFromArray(
       Array Sampling Mode=::NewEnumerator1
       Direct Array Index=$User.VariantNumber
       Color Selection Array=$User.VariantColor_B1_ColorArray
       Random Seed=?
      )
      Scale Factor=&MakeVector4_FromLinearColor(LinearColor=&FNTimeOfDayColor(Night Color=<rgba 0.5 0.5 0.5 1>, Morning Color=<rgba 1 1 1 1>, Day Color=<rgba 1 1 1 1>, Evening Color=<rgba 1 1 1 1>))
     ))
    ScaleA=true
    ScaleRGB=true
   )
   &InterpolateOverTime(Rate Of Change=10, Target Vector Value:NiagaraPosition=$Particles.Position, Variable Type=::NewEnumerator2)
   /SetVariables(Particles.Interpolate DY:Vector4f=&MakeFloat4FromVectorAndFloat("Vector (RGB)"=$Particles.InterpolateOverTime.InterpolateOverTime_V3_.Moving Average, "Float (Alpha)"=0.6)) {isSetParameters=true, module=SetVariables_#6}
   /SetVariables(Particles.Color1:LinearColor=&MakeLinearColorFromVectorAndFloat(
     "Vector (RGB)"=&Multiply_Vector(A=&MakeVectorFromLinearColorRGB(Color=$Particles.Color), B=<0.4 0.4 0.4>)
     "Float (Alpha)"=&Multiply_Float(A=1, B=$Particles.ADSFade)
    )) {isSetParameters=true, module=SetVariables_#7}
   /SetVariables(Particles.Color2=&MakeLinearColorFromVectorAndFloat(
     "Vector (RGB)"=&Multiply_Vector(A=&MakeVectorFromLinearColorRGB(Color=$Particles.Color), B=<0.6 0.6 0.6>)
     "Float (Alpha)"=&Multiply_Float(A=1, B=$Particles.ADSFade)
    )) {isSetParameters=true, module=SetVariables_#8}
   /SetVariables(Particles.Color3:LinearColor=&MakeLinearColorFromVectorAndFloat(
     "Vector (RGB)"=&Multiply_Vector(A=&MakeVectorFromLinearColorRGB(Color=$Particles.Color), B=<1 1 1>)
     "Float (Alpha)"=&Multiply_Float(A=2, B=$Particles.ADSFade)
    )) {isSetParameters=true}
   &DynamicMaterialParameters(
    Index 0 Param 1=&Lerp_Float(
     A=&SelectFloatFromArray(
      Direct Array Index=$Particles.PartitionParticles.Partition
      Float Selection Array=DataInterface() {propertyValues {FloatData [0.3,0.3,0.35,0.35,0.4,0.4]}}
     )
     B=0.8
     Alpha=0
    )
    Index 0 Param 2=$Particles.Timeline.Time
    Index 0 Param 3=&SelectFloatFromArray(Direct Array Index=$Particles.PartitionParticles.Partition, Float Selection Array=DataInterface() {propertyValues {FloatData [4,4,3,3,2,2]}})
    Index 0 Param 4=&SelectFloatFromArray(
     Direct Array Index=$Particles.PartitionParticles.Partition
     Float Selection Array=DataInterface() {propertyValues {FloatData [0,0.22,0.33,0.11,0.44,0.66]}}
    )
    Module.Index0Param0Write=true
    Module.Index0Param1Write=true
    Module.Index0Param2Write=true
    Module.Index0Param3Write=true
   )
   /SetVariables(
    Particles.GlowCameraOffset=2
    Particles.GlowColor=&MultiplyLinearColor(
     Linear Color=$Particles.Color
     Scale Factor=&MakeVector4_FromFloat(Float=&Multiply_Float(
       A=0.75
       B=&Waveform(
        "Waveform [1]"=::NewEnumerator12
        "[1] Phase Scale"=5
        "[1] Phase"=&Add_Float(
         A=$Emitter.Age
         B=&SelectFloatFromArray(
          Direct Array Index=$Particles.PartitionParticles.Partition
          Float Selection Array=DataInterface() {propertyValues {FloatData [0,0.11,0.22,0.175,0.33,0.55]}}
         )
        )
        "[1] Frequency"=2
        "[1] Amplitude Min/Max"=<0 1>
        "[1] Scale"=0.25
        "[1] Bias"=1
       )
      ))
    )
   ) {isSetParameters=true}
  ]
  rendererClasses [&NiagaraMeshRendererProperties]
  renderers [
   {rendererClass=&NiagaraMeshRendererProperties, rendererIndex=0}
   {rendererClass=&NiagaraMeshRendererProperties, rendererIndex=1}
   {rendererClass=&NiagaraMeshRendererProperties, rendererIndex=2}
   {rendererClass=&NiagaraMeshRendererProperties, rendererIndex=3}
  ]
 }
 Foot_Decal {
  emitterUpdateScript [
   &EmitterState(Life Cycle Mode=::NewEnumerator1, Loop Behavior=::NewEnumerator1, Loop Duration Mode=::NewEnumerator1, UseLoopDelay=false)
   &SpawnBurst_Instantaneous(Spawn Count=2, Use Spawn Probability=false, UseLoopCountLimit=false)
  ]
  particleSpawnScript [
   &PartitionParticles()
   &InitializeParticle(Sprite Size Mode=::NewEnumerator3, UsePositionOffset=false, Write Material Random=true)
   &CameraOffset(Camera Offset Amount=10)
   &DynamicMaterialParameters(
    Index 0 Param 1=1.5
    Index 1 Param 2=5
    Module.Index0Param0Write=true
    Module.Index0Param1Write=true
    Module.Index0Param2Write=true
    Module.Index0Param3Write=true
    Module.Index1Param0Write=true
    Module.Index1Param1Write=true
    Module.Index1Param2Write=true
    Module.Index1Param3Write=true
    Write Parameter Index 1=true
   )
  ]
  particleUpdateScript [
   &ParticleState(Kill Particles When Lifetime Has Elapsed=false, Let Infinitely Lived Particles Die When Emitter Deactivates=true)
   &ScaleColor(
    Scale Alpha=&MakeCustomFloatFromBool(Bool=$User.bisFootHidden, True Float=0, False Float=1)
    Scale RGB=&MakeVectorFromLinearColorRGB(Color=&MultiplyLinearColor(
      Linear Color=&SelectLinearColorFromArray(
       Array Sampling Mode=::NewEnumerator1
       Direct Array Index=$User.VariantNumber
       Color Selection Array=$User.VariantColor_B1_ColorArray
       Random Seed=?
      )
      Scale Factor=&MakeVector4_FromLinearColor(LinearColor=&FNTimeOfDayColor(Night Color=<rgba 1 1 1 1>, Morning Color=<rgba 1 1 1 1>, Day Color=<rgba 1 1 1 1>, Evening Color=<rgba 1 1 1 1>))
     ))
    ScaleA=true
    ScaleRGB=true
   )
   &SkeletalMeshLocation(
    Bone / Socket Index=&SelectIntFromArray(Direct Array Index=$Particles.PartitionParticles.Partition, Int Selection Array=DataInterface() {propertyValues {IntData [0,1]}})
    Bone Sampling Mode=::NewEnumerator4
    MaskSpawnGroup=false
    Sampled Position Offset=&SelectVectorFromArray(
     Direct Array Index=$Particles.PartitionParticles.Partition
     Vector Selection Array=DataInterface() {propertyValues {FloatData [<-2 -4 -0.5>, <2 4 0.5>]}}
    )
    Skeletal Mesh=$System.Skeletal Mesh
   )
  ]
  rendererClasses [&NiagaraSpriteRendererProperties]
  renderers [{rendererClass=&NiagaraSpriteRendererProperties}]
 }
 NeckFlame {
  emitterUpdateScript [
   &EmitterState(Life Cycle Mode=::NewEnumerator1, Loop Behavior=::NewEnumerator1, Loop Duration Mode=::NewEnumerator1, UseLoopDelay=false)
   &SpawnBurst_Instantaneous(Use Spawn Probability=false, UseLoopCountLimit=false)
   /SetVariables(
    Emitter.ColorA:LinearColor=&SelectLinearColorFromArray(
     Array Sampling Mode=::NewEnumerator1
     Direct Array Index=$User.VariantNumber
     Color Selection Array=$User.VariantColor_A1_ColorArray
     Random Seed=?
    )
    Emitter.ColorB:LinearColor=&SelectLinearColorFromArray(
     Array Sampling Mode=::NewEnumerator1
     Direct Array Index=$User.VariantNumber
     Color Selection Array=$User.VariantColor_A2_ColorArray
     Random Seed=?
    )
   ) {isSetParameters=true, module=SetVariables_#10}
  ]
  particleSpawnScript [
   &InitializeParticle(Lifetime=5, UsePositionOffset=false, Write Material Random=true)
   &CameraOffset(Camera Offset Amount=4)
  ]
  particleUpdateScript [
   &SkeletalMeshLocation(
    Bone / Socket Index=6
    Bone Sampling Mode=::NewEnumerator4
    MaskSpawnGroup=false
    Sampled Position Offset=<0 -1 0>
    Skeletal Mesh=$System.Skeletal Mesh
    Use Sampled Mesh Attribute as Ribbon ID=false
   )
   &ScaleColor(
    Scale Alpha=&SelectFloatFromArray(Direct Array Index=$User.VariantNumber, Float Selection Array=DataInterface() {propertyValues {FloatData [1,1.6,1]}})
    Scale RGB=&MakeVectorFromLinearColorRGB(Color=&MultiplyLinearColor(
      Linear Color=&SelectLinearColorFromArray(
       Array Sampling Mode=::NewEnumerator1
       Direct Array Index=$User.VariantNumber
       Color Selection Array=$User.VariantColor_A1_ColorArray
       Random Seed=?
      )
      Scale Factor=&MakeVector4_FromLinearColor(LinearColor=&FNTimeOfDayColor(Night Color=<rgba 0.8 0.8 0.8 1>, Morning Color=<rgba 0.9 0.9 0.9 1>, Day Color=<rgba 1 1 1 1>, Evening Color=<rgba 0.9 0.9 0.9 1>))
     ))
    ScaleA=true
    ScaleRGB=true
   )
   /SetVariables(Particles.GlowColor=&MultiplyLinearColor(Linear Color=$Particles.Color, Scale Factor=<1 1 1 0.5>)) {isSetParameters=true, module=SetVariables_#11}
   &ParticleState(Kill Particles When Lifetime Has Elapsed=false, Let Infinitely Lived Particles Die When Emitter Deactivates=true)
   &InterpolateOverTime(
    Rate Of Change:NiagaraFloat=$System.Interpolate Rate
    Target Vector Value:NiagaraPosition=$Particles.Position
    Variable Type=::NewEnumerator2
   )
   /SetVariables(Particles.Interpolate Dynamic=&MakeFloat4FromVectorAndFloat("Vector (RGB)"=$Particles.InterpolateOverTime.InterpolateOverTime_V3_.Moving Average, "Float (Alpha)"=$System.Interpolate Internsity)) {isSetParameters=true, module=SetVariables_#12}
   &DynamicMaterialParameters(
    Index 0 Param 1=0.35
    Index 0 Param 2=&Multiply_Float(A=$Engine.Time, B=2)
    Index 0 Param 3=0.5
    Index 0 Param 4=5.5921
    Module.Index0Param0Write=true
    Module.Index0Param1Write=true
    Module.Index0Param2Write=true
    Module.Index0Param3Write=true
   )
  ]
  rendererClasses [&NiagaraMeshRendererProperties]
  renderers [
   {rendererClass=&NiagaraMeshRendererProperties, rendererIndex=0}
   {rendererClass=&NiagaraMeshRendererProperties, rendererIndex=1}
   {rendererClass=&NiagaraMeshRendererProperties, rendererIndex=2}
   {rendererClass=&NiagaraMeshRendererProperties, rendererIndex=3}
  ]
 }
 NeckFlame001 {
  emitterUpdateScript [
   &EmitterState(Life Cycle Mode=::NewEnumerator1, Loop Behavior=::NewEnumerator1, Loop Duration Mode=::NewEnumerator1, UseLoopDelay=false)
   &SpawnBurst_Instantaneous(Spawn Count=2, Use Spawn Probability=false, UseLoopCountLimit=false)
   /SetVariables(
    Emitter.ColorA:LinearColor=&SelectLinearColorFromArray(
     Array Sampling Mode=::NewEnumerator1
     Direct Array Index=$User.VariantNumber
     Color Selection Array=$User.VariantColor_A1_ColorArray
     Random Seed=?
    )
    Emitter.ColorB:LinearColor=&SelectLinearColorFromArray(
     Array Sampling Mode=::NewEnumerator1
     Direct Array Index=$User.VariantNumber
     Color Selection Array=$User.VariantColor_A2_ColorArray
     Random Seed=?
    )
   ) {isSetParameters=true, module=SetVariables_#10}
  ]
  particleSpawnScript [
   &PartitionParticles()
   &InitializeParticle(Lifetime=5, Mesh Scale Mode=::NewEnumerator3, UsePositionOffset=false, Write Material Random=true)
   /SetVariables(Particles.MeshIndex=&SelectIntFromArray(Direct Array Index=$Particles.PartitionParticles.Partition, Int Selection Array=DataInterface() {propertyValues {IntData [0,1]}})) {isSetParameters=true}
   &CameraOffset(Camera Offset Amount=1)
   &InitialMeshOrientation(Mesh Orientation Mode=::NewEnumerator2, Rotation=<0.16 0 -0.01>, Use Rotation Vector=true)
  ]
  particleUpdateScript [
   &SkeletalMeshLocation(
    Bone / Socket Index=&SelectIntFromArray(Direct Array Index=$Particles.PartitionParticles.Partition, Int Selection Array=DataInterface() {propertyValues {IntData [9,10]}})
    Bone Sampling Mode=::NewEnumerator4
    MaskSpawnGroup=false
    Sampled Position Offset=&SelectVectorFromArray(
     Direct Array Index=$Particles.PartitionParticles.Partition
     Vector Selection Array=DataInterface() {propertyValues {FloatData [<-15 2 -7>, <15 -2 7>]}}
    )
    Skeletal Mesh=$System.Skeletal Mesh
    Use Sampled Mesh Attribute as Ribbon ID=false
   )
   &ScaleColor(
    Scale Alpha=&SelectFloatFromArray(Direct Array Index=$User.VariantNumber, Float Selection Array=DataInterface() {propertyValues {FloatData [1,1.6,1]}})
    Scale RGB=&MakeVectorFromLinearColorRGB(Color=&MultiplyLinearColor(
      Linear Color=&SelectLinearColorFromArray(
       Array Sampling Mode=::NewEnumerator1
       Direct Array Index=$User.VariantNumber
       Color Selection Array=$User.VariantColor_A1_ColorArray
       Random Seed=?
      )
      Scale Factor=&MakeVector4_FromLinearColor(LinearColor=&FNTimeOfDayColor(Night Color=<rgba 0.8 0.8 0.8 1>, Morning Color=<rgba 0.9 0.9 0.9 1>, Day Color=<rgba 1 1 1 1>, Evening Color=<rgba 0.9 0.9 0.9 1>))
     ))
    ScaleA=true
    ScaleRGB=true
   )
   /SetVariables(Particles.GlowColor=&MultiplyLinearColor(Linear Color=$Particles.Color, Scale Factor=<1 1 1 0.5>)) {isSetParameters=true, module=SetVariables_#11}
   &ParticleState(Kill Particles When Lifetime Has Elapsed=false, Loop Particles Lifetime=true)
   &InterpolateOverTime(
    Rate Of Change:NiagaraFloat=$System.Interpolate Rate
    Target Vector Value:NiagaraPosition=$Particles.Position
    Variable Type=::NewEnumerator2
   )
   /SetVariables(Particles.Interpolate Dynamic=&MakeFloat4FromVectorAndFloat("Vector (RGB)"=$Particles.InterpolateOverTime.InterpolateOverTime_V3_.Moving Average, "Float (Alpha)"=$System.Interpolate Internsity)) {isSetParameters=true, module=SetVariables_#12}
   &DynamicMaterialParameters(
    Index 0 Param 1=0.41
    Index 0 Param 2=&Multiply_Float(A=$Engine.Time, B=1.5)
    Module.Index0Param0Write=true
    Module.Index0Param1Write=true
    Module.Index0Param2Write=false
    Module.Index0Param3Write=false
    Module.Index1Param0Write=true
    Module.Index1Param1Write=true
    Module.Index1Param2Write=true
    Module.Index1Param3Write=true
    Write Parameter Index 1=true
   )
  ]
  rendererClasses [&NiagaraMeshRendererProperties]
  renderers [{rendererClass=&NiagaraMeshRendererProperties, rendererIndex=0}]
 }
 ShoulderFlameLoose {
  emitterUpdateScript [
   &EmitterState(
    Inactive Response=::NewEnumerator1
    Life Cycle Mode=::NewEnumerator1
    Loop Behavior=::NewEnumerator1
    Loop Duration Mode=::NewEnumerator1
    UseLoopDelay=false
   )
   &SpawnBurst_Instantaneous(Spawn Count=6, Use Spawn Probability=false, UseLoopCountLimit=false)
   /SetVariables(
    Emitter.ColorA:LinearColor=&SelectLinearColorFromArray(
     Array Sampling Mode=::NewEnumerator1
     Direct Array Index=$User.VariantNumber
     Color Selection Array=$User.VariantColor_A1_ColorArray
     Random Seed=?
    )
    Emitter.ColorB:LinearColor=&SelectLinearColorFromArray(
     Array Sampling Mode=::NewEnumerator1
     Direct Array Index=$User.VariantNumber
     Color Selection Array=$User.VariantColor_A2_ColorArray
     Random Seed=?
    )
   ) {isSetParameters=true, module=SetVariables_#10}
   &SpawnBurst_Instantaneous(Spawn Count=6, Spawn Time=0.973, Use Spawn Probability=false, UseLoopCountLimit=false) {module=SpawnBurst_Instantaneous001}
  ]
  particleSpawnScript [
   &PartitionParticles(Number Of Partitions=6)
   &InitializeParticle(
    Mesh Scale Mode=::NewEnumerator3
    Sprite Size Mode=::NewEnumerator3
    Uniform Sprite Size=12
    UsePositionOffset=false
    Write Material Random=true
   )
   &InitialMeshOrientation(
    Mesh Orientation Mode=::NewEnumerator2
    Rotation=&SelectVectorFromArray(
     Direct Array Index=$Particles.PartitionParticles.Partition
     Vector Selection Array=DataInterface() {propertyValues {FloatData [<1 0 0>, <1 0 0>, <1 0 0>, <0 0.5 0>, <0 0.5 0>, <0 0.5 0>]}}
    )
    Use Rotation Vector=true
   )
   &InitialMeshOrientation(
    Mesh Orientation Mode=::NewEnumerator2
    Rotation=&SelectVectorFromArray(
     Direct Array Index=$Particles.PartitionParticles.Partition
     Vector Selection Array=DataInterface() {propertyValues {FloatData [<0.25 -0.075 0>, <0.25 -0.075 0>, <0.25 -0.075 0>, <0.25 -0.075 0>, <0.25 -0.075 0>, <0.25 -0.075 0>]}}
    )
    Use Rotation Vector=true
   ) {module=InitialMeshOrientation001}
  ]
  particleUpdateScript [
   &SkeletalMeshLocation(
    Bone / Socket Index=&SelectIntFromArray(Direct Array Index=$Particles.PartitionParticles.Partition, Int Selection Array=DataInterface() {propertyValues {IntData [9,9,9,10,10,10]}})
    Bone Sampling Mode=::NewEnumerator4
    MaskSpawnGroup=false
    Sampled Position Offset=&SelectVectorFromArray(
     Direct Array Index=$Particles.PartitionParticles.Partition
     Vector Selection Array=DataInterface() {propertyValues {FloatData [<-3 -4 -16>, <4 -3 -8>, <0 -5 -12>, <3 4 16>, <-4 3 8>, <0 5 12>]}}
    )
    Skeletal Mesh=$System.Skeletal Mesh
   )
   /SetVariables(Particles.ADSFade:NiagaraFloat=&Lerp_Float(
     A=0
     B=1
     Alpha=&FN_LOD_Distance_Remap(Start Distance=250, End Distance=450, Minimum Output=0.5, Maximum Output=1, Clamp Output=true)
    )) {isSetParameters=true, module=SetVariables_#9}
   &ParticleState(Kill Particles When Lifetime Has Elapsed=false, Let Infinitely Lived Particles Die When Emitter Deactivates=true)
   &Timeline(
    Play Rate=&SelectFloatFromArray(
     Direct Array Index=$Particles.PartitionParticles.Partition
     Float Selection Array=DataInterface() {propertyValues {FloatData [2.5,2.5,2.2,2.2,1.75,1.75]}}
    )
    Timeline End=10000000
   )
   &ScaleColor(
    Scale Alpha=&SelectFloatFromArray(Direct Array Index=$User.VariantNumber, Float Selection Array=DataInterface() {propertyValues {FloatData [1,1.6,1]}})
    Scale RGB=&MakeVectorFromLinearColorRGB(Color=&MultiplyLinearColor(
      Linear Color=&SelectLinearColorFromArray(
       Array Sampling Mode=::NewEnumerator1
       Direct Array Index=$User.VariantNumber
       Color Selection Array=$User.VariantColor_A1_ColorArray
       Random Seed=?
      )
      Scale Factor=&MakeVector4_FromLinearColor(LinearColor=&FNTimeOfDayColor(Night Color=<rgba 0.8 0.8 0.8 1>, Morning Color=<rgba 0.9 0.9 0.9 1>, Day Color=<rgba 1 1 1 1>, Evening Color=<rgba 0.9 0.9 0.9 1>))
     ))
    ScaleA=true
    ScaleRGB=true
   )
   &ScaleMeshSize(Scale Factor=&Multiply_Vector(
     A=&Multiply_VectorByFloat(
      Vector=<0.2 0.2 0.2>
      Float=&SelectFloatFromArray(
       Direct Array Index=$Particles.PartitionParticles.Partition
       Float Selection Array=DataInterface() {propertyValues {FloatData [0.5,0.5,0.5,0.5,0.5,0.5]}}
      )
     )
     B=&Lerp_Vector(
      A=&Lerp_Vector(
       A=&Lerp_Vector(A=<1 1 1>, B=<1 1 1>, Alpha=$System.InterpolateOverTime.InterpolateOverTime_Float.Moving Average)
       B=<0.8 0.8 0.8>
       Alpha=$System.InterpolateOverTime.InterpolateOverTime_Float.Moving Average
      )
      B=<0.5 0.5 0.5>
      Alpha=&ClampFloat(Float=$System.InterpolateOverTime.InterpolateOverTime_Float.Moving Average)
     )
    ))
   &InterpolateOverTime(Rate Of Change=10, Target Vector Value:NiagaraPosition=$Particles.Position, Variable Type=::NewEnumerator2)
   /SetVariables(Particles.Interpolate DY:Vector4f=&MakeFloat4FromVectorAndFloat("Vector (RGB)"=$Particles.InterpolateOverTime.InterpolateOverTime_V3_.Moving Average, "Float (Alpha)"=$System.Interpolate Internsity)) {isSetParameters=true, module=SetVariables_#6}
   &DynamicMaterialParameters(
    Index 0 Param 1=&SelectFloatFromArray(
     Direct Array Index=$Particles.PartitionParticles.Partition
     Float Selection Array=DataInterface() {propertyValues {FloatData [0.7,0.5,0.5,0.7,0.5,0.5]}}
    )
    Index 0 Param 2=$Particles.Timeline.Time
    Index 0 Param 3=1
    Index 0 Param 4=&SelectFloatFromArray(
     Direct Array Index=$Particles.PartitionParticles.Partition
     Float Selection Array=DataInterface() {propertyValues {FloatData [0,0.33,0.66,0.33,0.55,0.77]}}
    )
    Module.Index0Param0Write=true
    Module.Index0Param1Write=true
    Module.Index0Param2Write=true
    Module.Index0Param3Write=true
   )
  ]
  rendererClasses [&NiagaraMeshRendererProperties]
  renderers [{rendererClass=&NiagaraMeshRendererProperties, rendererIndex=0}]
 }
 Sparks {
  simTarget=GPUComputeSim
  emitterUpdateScript [
   &EmitterState(Life Cycle Mode=::NewEnumerator1, Loop Duration=0.5, UseLoopDelay=false)
   &SpawnRate(Spawn Probability=0.33, SpawnRate=66, Use Spawn Probability=true)
  ]
  particleSpawnScript [
   &PartitionParticles()
   &InitializeParticle(
    Lifetime=&RandomRangeFloat(Minimum=0.3, Maximum=0.75)
    Sprite Size=<3 3>
    Sprite Size Mode=::NewEnumerator4
    UsePositionOffset=false
    Write Material Random=true
   )
   &SkeletalMeshLocation(
    Bone / Socket Index=&SelectIntFromArray(Direct Array Index=$Particles.PartitionParticles.Partition, Int Selection Array=DataInterface() {propertyValues {IntData [2,3]}})
    Bone Sampling Mode=::NewEnumerator4
    MaskSpawnGroup=false
    Orientation Sampling=::NewEnumerator1
    Sampled Position Offset=&SelectVectorFromArray(
     Direct Array Index=$Particles.PartitionParticles.Partition
     Vector Selection Array=DataInterface() {propertyValues {FloatData [<15 0 2>, <-15 0 -2>]}}
    )
    Skeletal Mesh=$System.Skeletal Mesh
   ) {module=SkeletalMeshLocation001}
   &ShapeLocation(MaskSpawnGroup=false, Override Seed=false, Sphere Radius=7, UseOwnerScale=true)
   &AddVelocity(
    Additional Quaternion Rotation=false
    Cone Axis=&SelectVectorFromArray(
     Direct Array Index=$Particles.PartitionParticles.Partition
     Vector Selection Array=DataInterface() {propertyValues {FloatData [<0 -1 1>, <0 1 -1>]}}
    )
    Module.UseConeFalloff=false
    Module.UseVelDistribution=false
    Override Seed=false
    Rotation Mode=::NewEnumerator3
    Rotation Quaternion:Quat4f=$Particles.SkeletalMeshLocation001.SampledRotation
    Velocity Mode=::NewEnumerator2
    Velocity Speed=50
   )
   &ApplyInitialForces()
  ]
  particleUpdateScript [
   &ParticleState()
   &ScaleSpriteSize(Uniform Curve Sprite Scale=DataInterface() {propertyValues {Curve {keys [
        <at 0 0 RCIM_Cubic>
        {interpMode=RCIM_Cubic, tangentMode=RCTM_User, time=0.25, value=1}
        {interpMode=RCIM_Cubic, tangentMode=RCTM_User, time=1, value=0}
       ]}}})
   &ScaleColor(
    Scale RGB=&MakeVectorFromLinearColorRGB(Color=&LerpLinearColors(
      StartColor=&MultiplyLinearColor(
       Linear Color=&SelectLinearColorFromArray(
        Array Sampling Mode=::NewEnumerator1
        Direct Array Index=$User.VariantNumber
        Color Selection Array=$User.VariantColor_A1_ColorArray
        Random Seed=?
       )
       Scale Factor=&MakeVector4_FromLinearColor(LinearColor=&FNTimeOfDayColor(Night Color=<rgba 0.7 0.7 0.7 1>, Morning Color=<rgba 0.9 0.9 0.9 1>, Day Color=<rgba 1 1 1 1>, Evening Color=<rgba 0.9 0.9 0.9 1>))
      )
      EndColor=&MultiplyLinearColor(
       Linear Color=&SelectLinearColorFromArray(
        Array Sampling Mode=::NewEnumerator1
        Direct Array Index=$User.VariantNumber
        Color Selection Array=$User.VariantColor_A2_ColorArray
        Random Seed=?
       )
       Scale Factor=&MakeVector4_FromLinearColor(LinearColor=&FNTimeOfDayColor(Night Color=<rgba 0.7 0.7 0.7 1>, Morning Color=<rgba 0.9 0.9 0.9 1>, Day Color=<rgba 1 1 1 1>, Evening Color=<rgba 0.9 0.9 0.9 1>))
      )
      LerpFactor=&FloatFromCurve(FloatCurve=DataInterface() {propertyValues {Curve {keys [{tangentMode=RCTM_User, time=0, value=0}, {tangentMode=RCTM_User, time=1, value=1}]}}})
     ))
    ScaleA=true
    ScaleRGB=true
   )
   /SetVariables(Particles.Glow03_Color=&MakeLinearColorFromVectorAndFloat(
     "Vector (RGB)"=&MakeVectorFromLinearColorRGB(Color=$Particles.Color)
     "Float (Alpha)"=&Multiply_Float(A=0.15, B=1)
    )) {isSetParameters=true}
   &InheritSourceMovement(
    Applied Position Scale=&Multiply_VectorByFloat(
     Vector=<1 1 1>
     Float=&FloatFromCurve(FloatCurve=DataInterface() {propertyValues {Curve {keys [
          {interpMode=RCIM_Cubic, tangentMode=RCTM_User, time=0.25, value=1}
          {arriveTangent=-1.7553, interpMode=RCIM_Cubic, leaveTangent=-1.7553, tangentMode=RCTM_User, time=1, value=0}
         ]}}})
    )
    ApplyForce=false
    ApplyPosition=true
    ApplyPositionFalloff=false
   )
   &CurlNoiseForce(Noise Frequency=250, Noise Strength=50, Pan Noise=false)
   &Drag(Drag=2, Use Linear Drag=true, Use Rotational Drag=false)
   &SolveForcesAndVelocity(Clamp Velocity=false, Limit Acceleration=false)
   &ScaleSpriteSizeBySpeed(Max Scale Factor=<0.25 1>, Min Scale Factor=<0.25 1>, Velocity Threshold=25)
   /SetVariables(Particles.DarkGlowSize:Vector2f=&Multiply_Vector2D(A=$Particles.SpriteSize, B=<5 2>)) {isSetParameters=true}
  ]
  rendererClasses [&NiagaraSpriteRendererProperties]
  renderers [
   {rendererClass=&NiagaraSpriteRendererProperties}
   {rendererClass=&NiagaraSpriteRendererProperties, rendererIndex=1}
  ]
 }
}
summary {
 systemName=NS_GateHound_BodyFlames
 emitters [
  {emitterName=NeckFlame, rendererClasses [&NiagaraMeshRendererProperties]}
  {emitterName=Fire_Spline_Feet, rendererClasses [&NiagaraMeshRendererProperties]}
  {emitterName=Foot_Decal, rendererClasses [&NiagaraSpriteRendererProperties]}
  {emitterName=Sparks, simTarget=GPUComputeSim, rendererClasses [&NiagaraSpriteRendererProperties]}
  {emitterName=Fire_Spline_Feet001, rendererClasses [&NiagaraMeshRendererProperties]}
  {emitterName=NeckFlame001, rendererClasses [&NiagaraMeshRendererProperties]}
  {emitterName=ShoulderFlameLoose, rendererClasses [&NiagaraMeshRendererProperties]}
 ]
 userVariables [
  User.bisFootHidden {type=NiagaraBool}
  User.bIsFrontEnd {type=NiagaraBool}
  User.bIsFrontEndPreview {type=NiagaraBool}
  User.bisHandHidden {type=NiagaraBool}
  User.IsSkydiving {type=NiagaraBool}
  User.VariantColor_A1_ColorArray {
   type=NiagaraDataInterfaceArrayColor
   description=""
   defaultValue {
    struct=NiagaraExt_VariableValue_DataInterface
    value {
     dataInterface=&self:NiagaraDataInterfaceArrayColor_0
     dataInterfaceClass=&NiagaraDataInterfaceArrayColor
    }
   }
  }
  User.VariantColor_A2_ColorArray {
   type=NiagaraDataInterfaceArrayColor
   description=""
   defaultValue {
    struct=NiagaraExt_VariableValue_DataInterface
    value {
     dataInterface=&self:NiagaraDataInterfaceArrayColor_2
     dataInterfaceClass=&NiagaraDataInterfaceArrayColor
    }
   }
  }
  User.VariantColor_B1_ColorArray {
   type=NiagaraDataInterfaceArrayColor
   description=""
   defaultValue {
    struct=NiagaraExt_VariableValue_DataInterface
    value {
     dataInterface=&self:NiagaraDataInterfaceArrayColor_1
     dataInterfaceClass=&NiagaraDataInterfaceArrayColor
    }
   }
  }
  User.VariantNumber {type=NiagaraInt32}
 ]
}
