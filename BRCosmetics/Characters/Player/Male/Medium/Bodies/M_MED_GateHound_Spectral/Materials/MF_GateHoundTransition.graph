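# MF_GateHoundTransition is a material function that adds a Styx emissive transition effect to the Spectral GateHound body cosmetic's material attributes.
# It builds panning main, detail and smoke noise, with normal-map UV offsets and optional world-aligned pre-skinned UVs.
# Masks come from emissive channels, an optional area mask, and pre-skinned distance from a neck position.
# Named reroutes carry these into a faded emissive term added to the EmissiveColor output.
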MaterialFunction /BRCosmetics/Characters/Player/Male/Medium/Bodies/M_MED_GateHound_Spectral/Materials/MF_GateHoundTransition
graph {nodes [
  n1 = Add(A=NamedRerouteUsage(), B=NamedRerouteUsage())
  n2 = NamedRerouteUsage()
  n3 = LinearInterpolate(
   A=LinearInterpolate(
    B=ScalarParameter() {defaultValue=4, group=Styx_Transit, parameterName=StyxEmissiveMulti}
    Alpha=LinearInterpolate(A=@n1, Alpha=@n2) {constB=0}
   ) {constA=1}
   B=Constant() {r=1}
   Alpha=NamedRerouteUsage()
  )
  n4 = NamedRerouteUsage()
  n5 = NamedRerouteUsage()
  n6 = ScalarParameter() {defaultValue=1, group=Styx_Transit, parameterName=FadeMulti}
  n7 = SmoothStep(
   Min=ScalarParameter() {group=Styx_Transit, parameterName=FadeMin}
   Max=ScalarParameter() {defaultValue=0.7, group=Styx_Transit, parameterName=FadeMax}
   Value=Saturate(None=Multiply(A=@n5, B=@n6))
  )
  n8 = Multiply(A=@n7, B=NamedRerouteUsage())
  n9 = Multiply(
   A=Multiply(A=Multiply(A=@n3, B=@n4), B=StaticSwitchParameter(True=@n8, False=@n7) {group=StyxMod, parameterName=UseAreaMask})
   B=NamedRerouteUsage()
  )
  n10 = NamedRerouteUsage()
  FunctionOutput(None=SetMaterialAttributes!(MaterialAttributes=NamedRerouteUsage(), EmissiveColor=Add(A=@n9, B=@n10)) {attributeSetTypes [#1]})
  n11 = ScalarParameter() {defaultValue=1, group=Styx_Transit, parameterName=EmissiveFadeMin}
  n12 = ScalarParameter() {defaultValue=1.5, group=Styx_Transit, parameterName=EmissiveFadeMax}
  n13 = NamedRerouteUsage()
  n14 = GetMaterialAttributes(None=FunctionInput() {inputName=Material, inputType=FunctionInput_MaterialAttributes}) {attributeGetTypes [#1]}
  n15 = ChannelMaskParameter(None=NamedRerouteDeclaration(None=@n14.EmissiveColor) {name=ORG_EM, nodeColor=<rgba 0 0.6016 1 1>}) {group=Styx_Transit, maskChannel=Green, parameterName=UseEmissiveAsMask, defaultValue=<rgba 0 1 0 0>}
  n16 = NamedRerouteDeclaration(None=FunctionInput() {inputName=EmissiveTextureSampler}) {name=EmissiveIN, nodeColor=<rgba 1 0.5625 0 1>}
  n17 = ChannelMaskParameter(None=@n16) {group=Styx_Transit, maskChannel=Green, parameterName=UseEmissiveAsMask, defaultValue=<rgba 0 1 0 0>}
  n18 = Saturate(None=StaticSwitchParameter(True=@n15.R, False=NamedRerouteDeclaration(None=@n17.R) {name=EmissiveMask, nodeColor=<rgba 1 0.1406 0 1>}) {group=Styx_Transit, parameterName=UseOrgEM})
  n19 = Multiply(A=OneMinus(None=@n18), B=@n18)
  NamedRerouteDeclaration(None=Multiply(A=LinearInterpolate(A=@n11, B=@n12, Alpha=@n13), B=OneMinus(None=@n19))) {name=EmissiveFadeMask, nodeColor=<rgba 1 0.5625 0 1>}
  n20 = Divide(
   A=StaticComponentMaskParameter(None=NamedRerouteUsage()) {defaultB=true, defaultR=true, parameterName=WorldUVChannel}
   B=ScalarParameter() {defaultValue=20, group=Styx_Transit, parameterName=WorldUVSize}
  )
  n21 = TextureCoordinate()
  n22 = ScalarParameter() {defaultValue=3, group=Styx, parameterName=StyxMainNoiseTilling}
  NamedRerouteDeclaration(None=StaticSwitchParameter(True=NamedRerouteDeclaration(None=@n20) {name=UVVis, nodeColor=<rgba 0.0312 1 0 1>}, False=Multiply(A=@n21, B=@n22)) {group=Styx_Transit, parameterName=UsePreskinnedAlignedUV}) {name=UV_withParallax_A, nodeColor=<rgba 1 0.6562 0 1>}
  n23 = NamedRerouteUsage()
  n24 = AppendVector(
   A=ScalarParameter() {group=Styx, parameterName=StyxNoise_Speed2X}
   B=ScalarParameter() {defaultValue=-0.1, group=Styx, parameterName=StyxNoise_Speed2Y}
  )
  n25 = TextureSample(UVs=Add(A=Panner(Coordinate=@n23, Speed=@n24), B=NamedRerouteUsage()), Tex=NamedRerouteUsage()) {samplerSource=SSM_Wrap_WorldGroupSettings}
  n26 = Power(Base=@n25.R, Exp=ScalarParameter() {defaultValue=5, group=Styx, parameterName=StyxNoise_Contrast2})
  n27 = Saturate(None=Power(Base=@n25.R, Exp=ScalarParameter() {defaultValue=5, group=Styx, parameterName=StyxNoise_Contrast}))
  NamedRerouteDeclaration(None=Multiply(A=LinearInterpolate(A=@n26, B=@n25.R, Alpha=@n27), B=ScalarParameter() {defaultValue=1, group=Styx, parameterName=MainNoiseIntensity})) {name=MainNoisePattern, nodeColor=<rgba 0 1 0.3203 1>}
  n28 = ScalarParameter() {group=Styx_Detail, parameterName=Styx_Detail_Noise_Low}
  n29 = ScalarParameter() {defaultValue=1, group=Styx_Detail, parameterName=Styx_Detail_Noise_High}
  n30 = Panner(
   Coordinate=Add(A=NamedRerouteUsage(), B=NamedRerouteUsage())
   Speed=AppendVector(
    A=ScalarParameter() {group=Styx_Detail, parameterName=Styx_Detail_Noise_SpeedX}
    B=ScalarParameter() {defaultValue=-0.1, group=Styx_Detail, parameterName=Styx_Detail_Noise_SpeedY}
   )
  )
  n31 = TextureSampleParameter2D(UVs=@n30) {
   group=Styx_Detail
   parameterName=StyxNoiseDetail
   samplerType=SAMPLERTYPE_Masks
   texture=/BRCosmetics/Characters/Player/Male/Medium/Bodies/M_MED_GateHound_Spectral/Textures/T_StyxGateHound_DetailNoise_M
  }
  NamedRerouteDeclaration(None=Saturate(None=LinearInterpolate(A=@n28, B=@n29, Alpha=@n31.R))) {name=DetailNoise, nodeColor=<rgba 1 0.0938 0 1>}
  n32 = FunctionInput() {bUsePreviewValueAsDefault=true, inputName=PreSkinnedLocation, sortPriority=2, previewValue=<1 0 0 1>}
  NamedRerouteDeclaration(None=@n32) {name=PreSkinned Loc, nodeColor=<rgba 0.5937 1 0 1>}
  NamedRerouteDeclaration(None=Saturate(None=@n31.G)) {name=MarcoNoiseMask, nodeColor=<rgba 0 1 0.8359 1>}
  n33 = Subtract(A=NamedRerouteUsage(), B=VectorParameter() {group=StyxMod, parameterName=NeckPOS, defaultValue=<rgba 0 -5 135 1>}.RGB)
  n34 = ComponentMask(None=@n33) {g=true}
  n35 = Saturate(None=Divide(A=Abs(None=@n34), B=ScalarParameter() {defaultValue=5, parameterName=StretchUnit}) {constB=3})
  NamedRerouteDeclaration(None=OneMinus(None=SmoothStep(Value=@n35) {constMin=0.3})) {name=StretchMask, nodeColor=<rgba 0.5469 1 0 1>}
  NamedRerouteDeclaration(None=SmoothStep(
    Max=ScalarParameter() {defaultValue=1, group=Styx_Transit, parameterName=AreaMax}
    Value=ChannelMaskParameter(None=@n16) {group=Styx_Transit, parameterName=AreaMaskChannel}.R
   )) {name=AreaMask, nodeColor=<rgba 1 0.7031 0 1>}
  n36 = TextureSampleParameter2D(UVs=@n30) {
   group=StyxMod
   parameterName=DistortSmoke
   samplerType=SAMPLERTYPE_Grayscale
   texture=/BRCosmetics/Characters/Player/Male/Medium/Bodies/M_MED_GateHound_Spectral/Textures/T_FractalDistortMask_GateHound
  }
  n37 = Multiply(A=@n36.RGB, B=ScalarParameter() {defaultValue=10, group=StyxMod, parameterName=SmokeAreaMulti}) {constB=5}
  NamedRerouteDeclaration(None=Multiply(A=Saturate(None=@n37), B=ScalarParameter() {defaultValue=0.5, group=StyxMod, parameterName=SmokeAreaIntensity})) {name=SmokeArea, nodeColor=<rgba 0 1 0.1797 1>}
  n38 = ScalarParameter() {group=Styx, parameterName=StyxNoise_SpeedX}
  n39 = ScalarParameter() {defaultValue=-0.1, group=Styx, parameterName=StyxNoise_SpeedY}
  n40 = TextureObjectParameter() {
   group=Styx
   parameterName=StyxDistortNormal
   samplerType=SAMPLERTYPE_Normal
   texture=/BRCosmetics/Characters/Player/Male/Medium/Bodies/M_MED_GateHound_Spectral/Textures/T_StyxGateHound_N
  }
  n41 = TextureSample(UVs=Panner(Coordinate=NamedRerouteUsage(), Speed=AppendVector(A=@n38, B=@n39)), Tex=@n40) {
   samplerSource=SSM_Wrap_WorldGroupSettings
   samplerType=SAMPLERTYPE_Normal
   texture=/BRCosmetics/Characters/Player/Male/Medium/Bodies/M_MED_DarkStance_Pater/Textures/T_StyxCosmetic_N
  }
  NamedRerouteDeclaration(None=ComponentMask(None=@n41.RGB) {g=true, r=true}) {name=Distortion_NormalNoise_A, nodeColor=<rgba 0.3828 0 1 1>}
  NamedRerouteDeclaration(None=ComponentMask(None=@n41.RGB) {g=true, r=true}) {name=Distortion_NormalNoise_B, nodeColor=<rgba 0.1016 0 1 1>}
  NamedRerouteDeclaration(None=@n14.MaterialAttributes) {name=MA, nodeColor=<rgba 0 1 0.8125 1>}
  n42 = ScalarParameter() {defaultValue=0.01, group=Styx, parameterName=Styx_Distort_UV}
  NamedRerouteDeclaration(None=Multiply(A=NamedRerouteUsage(), B=@n42)) {name=Distortion_Final_A, nodeColor=<rgba 0.2422 0 1 1>}
  NamedRerouteDeclaration(None=Multiply(A=@n42, B=NamedRerouteUsage())) {name=Distortion_Final_B, nodeColor=<rgba 1 0 0.8906 1>}
  n43 = ComponentMask(None=@n33) {b=true}
  n44 = SmoothStep(Value=Saturate(None=Divide(A=@n43) {constB=10})) {constMax=0.4}
  NamedRerouteDeclaration(None=StaticSwitchParameter(True=Constant() {r=1}, False=OneMinus(None=@n44)) {group=Styx_Transit, parameterName=IsNeck}) {name=NeckMask, nodeColor=<rgba 1 0 0.2344 1>}
  n45 = TextureObjectParameter() {
   group=Styx
   parameterName=Styx Mask
   samplerType=SAMPLERTYPE_Masks
   texture=/BRCosmetics/Characters/Player/Male/Medium/Bodies/M_MED_GateHound_Spectral/Textures/T_StyxGateHound_Mask
  }
  NamedRerouteDeclaration(None=@n45) {name=Styx Mask IN, nodeColor=<rgba 1 0.8906 0 1>}
  NamedRerouteUsage()
 ]}
