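# Parent character material for the spectral GateHound male medium body skin, using material attributes with a shading model set by expression.
# It wraps MF_GateHoundTransition between the shared 2023 character PreFX and PostFX functions.
# The transition function takes the PreFX output plus an EmissiveRE texture parameter, defaulting to the spectral body emissive map.
# PostFX receives the transitioned material alongside pass-through inputs from PreFX.
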Material /BRCosmetics/Characters/Player/Male/Medium/Bodies/M_MED_GateHound_Spectral/Materials/M_GateHoundSpectral_Parent
graph {
 properties {
  bUseMaterialAttributes=true
  bUsedWithClothing=true
  bUsedWithMeshDeformer=true
  bUsedWithMorphTargets=true
  bUsedWithNanite=true
  bUsedWithSkeletalMesh=true
  pixelDepthOffsetMode=PDOM_Legacy
  shadingModel=MSM_FromMaterialExpression
  usedShadingModels=Default Lit | Subsurface | Clear Coat | Subsurface Profile | Cloth
 }
 nodes [
  n1 = TextureSampleParameter2D() {
   group=Styx
   parameterName=EmissiveRE
   texture=/BRCosmetics/Characters/Player/Male/Medium/Bodies/M_MED_GateHound_Spectral/Textures/T_GateHound_Spectral_Body_E
  }
  n2 = MaterialFunctionCall() {materialFunction=&MF_CharacterParent_2023_PreFX}
  MaterialFunctionCall(
   Material=MaterialFunctionCall!(EmissiveTextureSampler=@n1.RGB, Material=@n2.Result, PreSkinnedLocation=@n2.Result) {materialFunction=&/MF_GateHoundTransition}.Result
   M=@n2.M
   Pre Skinned Local Position=@n2.M
   Distance Blend=@n2.M
   Fresnel=@n2.M
  ) {materialFunction=&MF_CharacterParent_2023_PostFX}
 ]
}
