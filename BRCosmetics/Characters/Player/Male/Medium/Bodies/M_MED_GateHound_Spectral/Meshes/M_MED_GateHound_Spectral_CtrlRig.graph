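# Control Rig for the M_MED_GateHound_Spectral character body that corrects deformation of secondary bones as the body poses.
# Forward solve runs the shared DeformRig function, then shoulder armor ("Dog Head"), jacket, collar and belt buckle functions.
# Weighted parent constraints seat armor, jacket and buckle bones, while spherical pose readers and DeformRig anim attributes weight corrective offsets.
# IsSkydiveFloating blends in an extra jacket adjustment through an interpolated alpha.
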ControlRigBlueprint /BRCosmetics/Characters/Player/Male/Medium/Bodies/M_MED_GateHound_Spectral/Meshes/M_MED_GateHound_Spectral_CtrlRig
graphs [
 New Function_ContainedGraph {nodes [
   RigVMFunctionEntryNode()
   > RigVMFunction_Sequence = RigVMAggregateNode()
   &RigUnit_ParentConstraint(
    AdvancedSettings=(RotationOrderForFilter=XZY)
    Child=(Type=Bone,Name="deform_shoulder_armor_lower_l_base")
    Parents="((Item=(Type=Bone,Name=\"spine_03\"),Weight=0.0228),(Item=(Type=Bone,Name=\"spine_04\"),Weight=0.0795),(Item=(Type=Bone,Name=\"deform_upperarm_up_l\"),Weight=0.0826),(Item=(Type=Bone,Name=\"upperarm_twist_01_l\"),Weight=0.098),(Item=(Type=Bone,Name=\"deform_upperar ...(truncated)"
    Weight=1
    ExecutePin=@RigVMFunction_Sequence.A
   )
   > &RigUnit_ParentConstraint(
    AdvancedSettings=(RotationOrderForFilter=XZY)
    Child=(Type=Bone,Name="deform_shoulder_armor_r_base")
    Parents="((Item=(Type=Bone,Name=\"spine_03\"),Weight=0.0224),(Item=(Type=Bone,Name=\"spine_04\"),Weight=0.0792),(Item=(Type=Bone,Name=\"deform_upperarm_up_r\"),Weight=0.0832),(Item=(Type=Bone,Name=\"upperarm_twist_01_r\"),Weight=0.099),(Item=(Type=Bone,Name=\"deform_upperar ...(truncated)"
    Weight=1
   )
   > &RigUnit_ModifyTransforms(
    ItemToModify=((Item=(Type=Bone,Name="deform_shoulder_armor_lower_l"),Transform=(Rotation=(X=0,Y=0,Z=0,W=1),Translation=(X=2,Y=0,Z=0),Scale3D=(X=1,Y=1,Z=1))))
    WeightMinimum=0
    Weight=RigVMDispatchNode(BoneName=root, Default=0, Name=deformRig_l_upperarm_fr_alpha) {template="DISPATCH_RigDispatch_GetAnimAttribute(in Name,in BoneName,in Default,out Value,out Found)"}.Value
   )
   > &RigUnit_ModifyTransforms(
    ItemToModify=((Item=(Type=Bone,Name="deform_shoulder_armor_lower_r"),Transform=(Rotation=(X=0,Y=0,Z=0.0311,W=0.9995),Translation=(X=0,Y=0,Z=0),Scale3D=(X=1,Y=1,Z=1))))
    WeightMinimum=0
    Weight=RigVMDispatchNode(BoneName=root, Default=0, Name=deformRig_r_upperarm_fr_alpha) {template="DISPATCH_RigDispatch_GetAnimAttribute(in Name,in BoneName,in Default,out Value,out Found)"}.Value
   )
   > &RigUnit_ModifyTransforms(
    ItemToModify=((Item=(Type=Bone,Name="deform_shoulder_armor_lower_l"),Transform=(Rotation=(X=-0.1261,Y=-0.0338,Z=0.2566,W=0.9577),Translation=(X=2.5,Y=0,Z=2),Scale3D=(X=1,Y=1,Z=1))))
    WeightMinimum=0.3
    Weight=RigVMDispatchNode(BoneName=root, Default=0, Name=deformRig_l_upperarm_up_alpha) {template="DISPATCH_RigDispatch_GetAnimAttribute(in Name,in BoneName,in Default,out Value,out Found)"}.Value
   )
   > &RigUnit_ModifyTransforms(
    ItemToModify=((Item=(Type=Bone,Name="deform_shoulder_armor_lower_r"),Transform=(Rotation=(X=-0.1261,Y=-0.0338,Z=0.2566,W=0.9577),Translation=(X=-2.5,Y=0,Z=-2),Scale3D=(X=1,Y=1,Z=1))))
    WeightMinimum=0.3
    Weight=RigVMDispatchNode(BoneName=root, Default=0, Name=deformRig_r_upperarm_up_alpha) {template="DISPATCH_RigDispatch_GetAnimAttribute(in Name,in BoneName,in Default,out Value,out Found)"}.Value
   )
   n1 = &RigUnit_SphericalPoseReader(
    ActiveRegionScaleFactors=(PositiveWidth=1,NegativeWidth=1,PositiveHeight=1,NegativeHeight=1)
    ActiveRegionSize=0.1
    Debug=(bDrawDebug=false,bDraw2D=False,bDrawLocalAxes=False,DebugScale=25,DebugSegments=20,DebugThickness=0.25)
    DriverAxis=(X=1,Y=0,Z=0)
    DriverItem=(Type=Bone,Name="upperarm_l")
    FalloffRegionScaleFactors=(PositiveWidth=1,NegativeWidth=1,PositiveHeight=1,NegativeHeight=1)
    FalloffSize=0.2
    FlipHeightScaling=False
    FlipWidthScaling=False
    OptionalParentItem=(Type=Bone,Name="None")
    RotationOffset=(X=45,Y=0,Z=45)
    ExecutePin=@RigVMFunction_Sequence.B
   )
   > n2 = &RigUnit_SphericalPoseReader(
    ActiveRegionScaleFactors=(PositiveWidth=1,NegativeWidth=1,PositiveHeight=1,NegativeHeight=1)
    ActiveRegionSize=0.1
    Debug=(bDrawDebug=false,bDraw2D=False,bDrawLocalAxes=False,DebugScale=25,DebugSegments=20,DebugThickness=0.25)
    DriverAxis=(X=-1,Y=0,Z=0)
    DriverItem=(Type=Bone,Name="upperarm_r")
    FalloffRegionScaleFactors=(PositiveWidth=1,NegativeWidth=1,PositiveHeight=1,NegativeHeight=1)
    FalloffSize=0.3
    FlipHeightScaling=False
    FlipWidthScaling=False
    OptionalParentItem=(Type=Bone,Name="None")
    RotationOffset=(X=-130,Y=0,Z=45)
   )
   > &RigUnit_ModifyTransforms(
    ItemToModify=((Item=(Type=Bone,Name="deform_shoulder_armor_lower_l"),Transform=(Rotation=(X=0,Y=0,Z=0,W=1),Translation=(X=0,Y=0,Z=0),Scale3D=(X=1,Y=1,Z=1))))
    WeightMinimum=0
    Weight=@n1.OutputParam
   )
   > &RigUnit_ModifyTransforms(
    ItemToModify=((Item=(Type=Bone,Name="deform_shoulder_armor_lower_r"),Transform=(Rotation=(X=-0.0069,Y=-0.1133,Z=-0.0607,W=0.9917),Translation=(X=2.2324,Y=0,Z=-1.0141),Scale3D=(X=1,Y=1,Z=1))))
    WeightMinimum=0
    Weight=@n2.OutputParam
   )
   RigVMFunctionReturnNode(ExecuteContext=@RigVMFunction_Sequence.C)
  ]}
 New Function_ContainedGraph {nodes [
   n1 = RigVMFunctionEntryNode()
   RigVMFunction_Sequence = &RigVMFunction_Sequence(ExecuteContext=@n1.ExecuteContext)
   &RigUnit_ParentConstraint(
    AdvancedSettings=(RotationOrderForFilter=XZY)
    Child=(Type=Bone,Name="dyn_jacket_front_l_01_base")
    Parents="((Item=(Type=Bone,Name=\"pelvis\"),Weight=0.0113),(Item=(Type=Bone,Name=\"upperarm_twist_02_l\"),Weight=0.0247),(Item=(Type=Bone,Name=\"deform_upperarm_fr_l\"),Weight=0.0408),(Item=(Type=Bone,Name=\"spine_01\"),Weight=0.0488),(Item=(Type=Bone,Name=\"upperarm_twist_ ...(truncated)"
    Weight=1
    ExecutePin=@RigVMFunction_Sequence.A
   )
   > &RigUnit_ParentConstraint(
    AdvancedSettings=(RotationOrderForFilter=XZY)
    Child=(Type=Bone,Name="dyn_jacket_side_l_01_base")
    Parents="((Item=(Type=Bone,Name=\"deform_pec_l\"),Weight=0.0114),(Item=(Type=Bone,Name=\"pelvis\"),Weight=0.0148),(Item=(Type=Bone,Name=\"spine_01\"),Weight=0.0539),(Item=(Type=Bone,Name=\"clavicle_l\"),Weight=0.1425),(Item=(Type=Bone,Name=\"spine_04\"),Weight=0.1432),(Item= ...(truncated)"
    Weight=1
   )
   > &RigUnit_ParentConstraint(
    AdvancedSettings=(RotationOrderForFilter=XZY)
    Child=(Type=Bone,Name="dyn_jacket_rear_l_01_base")
    Parents="((Item=(Type=Bone,Name=\"spine_05\"),Weight=0.0034),(Item=(Type=Bone,Name=\"pelvis\"),Weight=0.0065),(Item=(Type=Bone,Name=\"spine_01\"),Weight=0.0305),(Item=(Type=Bone,Name=\"spine_02\"),Weight=0.0887),(Item=(Type=Bone,Name=\"spine_04\"),Weight=0.1371),(Item=(Type= ...(truncated)"
    Weight=1
   )
   > &RigUnit_ParentConstraint(
    AdvancedSettings=(RotationOrderForFilter=XZY)
    Child=(Type=Bone,Name="dyn_jacket_rear_01_base")
    Parents="((Item=(Type=Bone,Name=\"deform_upperarm_bk_l\"),Weight=0.0004),(Item=(Type=Bone,Name=\"deform_upperarm_bk_r\"),Weight=0.0004),(Item=(Type=Bone,Name=\"clavicle_l\"),Weight=0.011),(Item=(Type=Bone,Name=\"clavicle_r\"),Weight=0.011),(Item=(Type=Bone,Name=\"pelvis\"),W ...(truncated)"
    Weight=1
   )
   > &RigUnit_ParentConstraint(
    AdvancedSettings=(RotationOrderForFilter=XZY)
    Child=(Type=Bone,Name="dyn_jacket_side_r_01_base")
    Parents="((Item=(Type=Bone,Name=\"deform_upperarm_fr_r\"),Weight=0.0049),(Item=(Type=Bone,Name=\"pelvis\"),Weight=0.0148),(Item=(Type=Bone,Name=\"spine_01\"),Weight=0.0537),(Item=(Type=Bone,Name=\"clavicle_r\"),Weight=0.1424),(Item=(Type=Bone,Name=\"spine_04\"),Weight=0.143) ...(truncated)"
    Weight=1
   )
   > &RigUnit_ParentConstraint(
    AdvancedSettings=(RotationOrderForFilter=XZY)
    Child=(Type=Bone,Name="dyn_jacket_rear_r_01_base")
    Parents="((Item=(Type=Bone,Name=\"spine_05\"),Weight=0.0034),(Item=(Type=Bone,Name=\"pelvis\"),Weight=0.0065),(Item=(Type=Bone,Name=\"spine_01\"),Weight=0.0302),(Item=(Type=Bone,Name=\"spine_02\"),Weight=0.0881),(Item=(Type=Bone,Name=\"spine_04\"),Weight=0.1364),(Item=(Type= ...(truncated)"
    Weight=1
   )
   > &RigUnit_ParentConstraint(
    AdvancedSettings=(RotationOrderForFilter=XZY)
    Child=(Type=Bone,Name="dyn_jacket_front_r_01_base")
    Parents="((Item=(Type=Bone,Name=\"pelvis\"),Weight=0.0112),(Item=(Type=Bone,Name=\"upperarm_twist_02_r\"),Weight=0.0251),(Item=(Type=Bone,Name=\"deform_upperarm_fr_r\"),Weight=0.0415),(Item=(Type=Bone,Name=\"spine_01\"),Weight=0.0487),(Item=(Type=Bone,Name=\"upperarm_twist_ ...(truncated)"
    Weight=1
   )
   > &RigUnit_ModifyTransforms(
    ItemToModify="((Item=(Type=Bone,Name=\"dyn_jacket_rear_01_base\"),Transform=(Rotation=(X=0,Y=0,Z=0,W=1),Translation=(X=0,Y=-1,Z=0),Scale3D=(X=1,Y=1,Z=1))),(Item=(Type=Bone,Name=\"dyn_jacket_rear_r_01_base\"),Transform=(Rotation=(X=0,Y=0,Z=0,W=1),Translation=(X=0,Y=1,Z=0),Sc ...(truncated)"
    Weight=1
    WeightMinimum=0
   )
   > &RigUnit_ModifyTransforms(
    ItemToModify="((Item=(Type=Bone,Name=\"dyn_jacket_side_l_01_base\"),Transform=(Rotation=(X=0,Y=0,Z=0,W=1),Translation=(X=-2,Y=0,Z=0),Scale3D=(X=1,Y=1,Z=1))),(Item=(Type=Bone,Name=\"dyn_jacket_front_l_01_base\"),Transform=(Rotation=(X=0,Y=0,Z=0,W=1),Translation=(X=-3.5,Y=0,Z ...(truncated)"
    WeightMinimum=0
    Weight=RigVMDispatchNode(BoneName=root, Default=0, Name=deformRig_l_upperarm_up_alpha) {template="DISPATCH_RigDispatch_GetAnimAttribute(in Name,in BoneName,in Default,out Value,out Found)"}.Value
   )
   > &RigUnit_ModifyTransforms(
    ItemToModify="((Item=(Type=Bone,Name=\"dyn_jacket_side_r_01_base\"),Transform=(Rotation=(X=0,Y=0,Z=0,W=1),Translation=(X=2,Y=0,Z=0),Scale3D=(X=1,Y=1,Z=1))),(Item=(Type=Bone,Name=\"dyn_jacket_front_r_01_base\"),Transform=(Rotation=(X=0,Y=0,Z=0,W=1),Translation=(X=3.5,Y=0,Z=0 ...(truncated)"
    WeightMinimum=0
    Weight=RigVMDispatchNode(BoneName=root, Default=0, Name=deformRig_r_upperarm_up_alpha) {template="DISPATCH_RigDispatch_GetAnimAttribute(in Name,in BoneName,in Default,out Value,out Found)"}.Value
   )
   RigVMFunction_MathBoolToFloat = &RigVMFunction_MathBoolToFloat(Value=RigVMVariableNode() {variable=IsSkydiveFloating}.Value)
   AlphaInterp = &RigVMFunction_AlphaInterp(InterpSpeedDecreasing=0.05, InterpSpeedIncreasing=0.05, Scale=1, bClampResult=true, Value=@RigVMFunction_MathBoolToFloat.Result)
   > &RigUnit_ModifyTransforms(
    ItemToModify="((Item=(Type=Bone,Name=\"dyn_jacket_front_l_01_base\"),Transform=(Rotation=(X=0,Y=0,Z=-0.2588,W=0.9659),Translation=(X=0,Y=0,Z=0),Scale3D=(X=1,Y=1,Z=1))),(Item=(Type=Bone,Name=\"dyn_jacket_front_r_01_base\"),Transform=(Rotation=(X=0,Y=0,Z=-0.2588,W=0.9659),Tra ...(truncated)"
    WeightMinimum=0
    Weight=@AlphaInterp.Result
   )
   RigVMFunctionReturnNode(ExecuteContext=@RigVMFunction_Sequence.B)
  ]}
 New Function_ContainedGraph {nodes [
   RigVMFunctionEntryNode()
   > n1 = &RigUnit_SphericalPoseReader(
    ActiveRegionScaleFactors=(PositiveWidth=1,NegativeWidth=3,PositiveHeight=3,NegativeHeight=3)
    ActiveRegionSize=0.1
    Debug=(bDrawDebug=false,bDraw2D=False,bDrawLocalAxes=False,DebugScale=25,DebugSegments=20,DebugThickness=0.25)
    DriverAxis=(X=1,Y=0,Z=0)
    DriverItem=(Type=Bone,Name="head")
    FalloffRegionScaleFactors=(PositiveWidth=1,NegativeWidth=1,PositiveHeight=1,NegativeHeight=1)
    FalloffSize=0.4
    FlipHeightScaling=False
    FlipWidthScaling=False
    OptionalParentItem=(Type=Bone,Name="spine_05")
    RotationOffset=(X=-90,Y=0,Z=-15)
   )
   RigVMFunction_AnimEvalRichCurve = &RigVMFunction_AnimEvalRichCurve(
    Curve=(EditorCurveData=(Keys=((InterpMode=RCIM_Cubic),(InterpMode=RCIM_Cubic,Time=0.7221,Value=0.448,ArriveTangent=1,LeaveTangent=1),(Time=1,Value=1))))
    SourceMinimum=0
    TargetMinimum=0
    Value=@n1.OutputParam
   )
   > &RigUnit_ModifyTransforms(
    ItemToModify="((Item=(Type=Bone,Name=\"dyn_collar_l_01_base\"),Transform=(Rotation=(X=0,Y=0,Z=0,W=1),Translation=(X=0,Y=0,Z=0),Scale3D=(X=1,Y=1,Z=1))),(Item=(Type=Bone,Name=\"dyn_collar_mid_01_base\"),Transform=(Rotation=(X=-0.0089,Y=-0.0347,Z=-0.4957,W=0.8677),Translation= ...(truncated)"
    WeightMinimum=0
    Weight=@RigVMFunction_AnimEvalRichCurve.Result
   )
   > n2 = &RigUnit_SphericalPoseReader(
    ActiveRegionScaleFactors=(PositiveWidth=1,NegativeWidth=2.3,PositiveHeight=2,NegativeHeight=1)
    ActiveRegionSize=0.1
    Debug=(bDrawDebug=false,bDraw2D=False,bDrawLocalAxes=False,DebugScale=25,DebugSegments=20,DebugThickness=0.25)
    DriverAxis=(X=1,Y=0,Z=0)
    DriverItem=(Type=Bone,Name="head")
    FalloffRegionScaleFactors=(PositiveWidth=1,NegativeWidth=1,PositiveHeight=1,NegativeHeight=1)
    FalloffSize=0.32
    FlipHeightScaling=False
    FlipWidthScaling=False
    OptionalParentItem=(Type=Bone,Name="spine_05")
    RotationOffset=(X=-40,Y=0,Z=50)
   )
   n3 = &RigVMFunction_AnimEvalRichCurve(
    Curve=(EditorCurveData=(Keys=((InterpMode=RCIM_Cubic),(InterpMode=RCIM_Cubic,Time=0.7221,Value=0.448,ArriveTangent=1,LeaveTangent=1),(Time=1,Value=1))))
    SourceMinimum=0
    TargetMinimum=0
    Value=@n2.OutputParam
   )
   > &RigUnit_ModifyTransforms(
    ItemToModify="((Item=(Type=Bone,Name=\"dyn_collar_l_01_base\"),Transform=(Rotation=(X=0,Y=0,Z=0,W=1),Translation=(X=0,Y=0,Z=0),Scale3D=(X=1,Y=1,Z=1))),(Item=(Type=Bone,Name=\"dyn_collar_mid_01_base\"),Transform=(Rotation=(X=0.1502,Y=-0.0402,Z=-0.3591,W=0.9203),Translation=( ...(truncated)"
    WeightMinimum=0
    Weight=@n3.Result
   )
   > RigVMFunctionReturnNode()
  ]}
 New Function_ContainedGraph {nodes [
   RigVMFunctionEntryNode()
   > &RigUnit_ParentConstraint(
    AdvancedSettings=(RotationOrderForFilter=XZY)
    Child=(Type=Bone,Name="deform_belt_buckle")
    Parents=((Item=(Type=Bone,Name="spine_04"),Weight=0.0026),(Item=(Type=Bone,Name="spine_03"),Weight=0.0709),(Item=(Type=Bone,Name="spine_01"),Weight=0.1011),(Item=(Type=Bone,Name="spine_02"),Weight=0.1187),(Item=(Type=Bone,Name="pelvis"),Weight=0.7067))
    Weight=1
   )
   > n1 = &RigUnit_SphericalPoseReader(
    ActiveRegionScaleFactors=(PositiveWidth=0.7,NegativeWidth=3,PositiveHeight=3,NegativeHeight=3)
    ActiveRegionSize=0.1
    Debug=(bDrawDebug=false,bDraw2D=False,bDrawLocalAxes=False,DebugScale=25,DebugSegments=20,DebugThickness=0.25)
    DriverAxis=(X=1,Y=0,Z=0)
    DriverItem=(Type=Bone,Name="spine_05")
    FalloffRegionScaleFactors=(PositiveWidth=1,NegativeWidth=1,PositiveHeight=1,NegativeHeight=1)
    FalloffSize=0.3
    FlipHeightScaling=False
    FlipWidthScaling=False
    OptionalParentItem=(Type=Bone,Name="pelvis")
    RotationOffset=(X=90,Y=0,Z=-35)
   )
   > &RigUnit_ModifyTransforms(
    ItemToModify=((Item=(Type=Bone,Name="deform_belt_buckle"),Transform=(Rotation=(X=0,Y=0,Z=-0.1736,W=0.9848),Translation=(X=0,Y=0,Z=0),Scale3D=(X=1,Y=1,Z=1))))
    WeightMinimum=0
    Weight=@n1.OutputParam
   )
   > RigVMFunctionReturnNode()
  ]}
 RigVMFunctionLibrary {nodes [
   Belt Buckle = RigVMCollapseNode(ExecuteContext=())
   Collar Deformation = RigVMCollapseNode(ExecuteContext=())
   Dog Head Deformation = RigVMCollapseNode(ExecuteContext=())
   Jacket Deformation = RigVMCollapseNode(ExecuteContext=())
  ]}
 RigVMFunction_Sequence_1_ContainedGraph {nodes [
   n1 = RigVMFunctionEntryNode()
   RigVMFunction_Sequence = &RigVMFunction_Sequence(ExecuteContext=@n1.ExecuteContext)
   n2 = &RigVMFunction_Sequence(ExecuteContext=@RigVMFunction_Sequence.B)
   RigVMFunctionReturnNode(A=@RigVMFunction_Sequence.A, B=@n2.A, C=@n2.B)
  ]}
 RigVMModel {nodes [
   RigUnit_BeginExecution = &RigUnit_BeginExecution()
   FN_DeformRig_FNC = RigVMFunctionReferenceNode(
    DeformRig_ Curve="(deformRig_l_upperarm_up_alpha_72_#1=True,deformRig_r_upperarm_up_alpha_73_#2=True,deformRig_l_upperarm_fr_alpha_74_#3=True,deformRig_r_upperarm_fr_alpha_75_CB51B80C4 ...(truncated)"
    DeformRig_Bool="(upperarm_1_#4=True,elbow_3_#5=True,wrist_7_#6=True,groin_8_#7=True,glute_17_#8=True,knee_9_40C408324CA261 ...(truncated)"
    DeformRig_Debug=(upperarm_debug_18_#4=False,wrist_debug_20_#6=False,lat_debug_26_#9=False,groinGlute_debug_27_#7=False)
    DeformRig_Value="(upperarm_up_18_#10=(Rotation=(X=0,Y=0,Z=0,W=1),Translation=(X=0,Y=0,Z=8),Scale3D=(X=0.7,Y=1,Z=1)),upperarm_fwd_fr_19_#11=(Rotation=(X=0,Y=0,Z=0,W=1),Translation=(X=0,Y=10,Z=0),Scale3D=(X=0.5,Y=1,Z= ...(truncated)"
    ExecuteContext=@RigUnit_BeginExecution.ExecutePin
   )
   > Dog Head Deformation = RigVMFunctionReferenceNode()
   > Jacket Deformation = RigVMFunctionReferenceNode()
   > Collar Deformation = RigVMFunctionReferenceNode()
   > Belt Buckle = RigVMFunctionReferenceNode()
  ]}
]
hierarchy [
 root {loc=<0 0 0>, rot=<0 0 0 1>}
 pelvis {parent=root, loc=<0 -0.7071 93.1191>, rot=<0.0127 0.707 -0.0127 -0.707>}
 spine_01 {parent=pelvis, loc=<6.7494 -0.6388 0>, rot=<0 0 0.034 -0.9994>}
 spine_02 {parent=spine_01, loc=<7.2518 -0.035 0>, rot=<0 0 0.0238 -0.9997>}
 spine_03 {parent=spine_02, loc=<6.0104 0 0>, rot=<0 0 0.0149 -0.9999>}
 spine_04 {parent=spine_03, loc=<8.1024 0 0>, rot=<0 0 -0.0132 -0.9999>}
 spine_05 {parent=spine_04, loc=<10.0146 0 0>, rot=<0 0 -0.0015 -1>}
 attach_cape {parent=spine_05, loc=<9.017 -11.714 0>, rot=<0.0616 -0.7044 -0.0616 -0.7044>}
 neck_01 {parent=spine_05, loc=<11.3138 0.1529 0>, rot=<0 0 -0.2003 -0.9797>}
 neck_02 {parent=neck_01, loc=<4.5627 0 0>, rot=<0 0 0.0504 -0.9987>}
 head {parent=neck_02, loc=<5.3914 0 0>, rot=<0 0 0.115 -0.9934>}
 hair_lower_root {parent=head, loc=<0 0 0>, rot=<0 0 0 1>}
 hair_lower_right_root {parent=hair_lower_root, loc=<0 0 0>, rot=<0 0 0 1>}
 attach_backpack {parent=spine_05, loc=<-0.123 -13.473 0>, rot=<-0.0093 -0.707 0.0093 -0.707>}
 clavicle_l {parent=spine_05, loc=<4.3575 -0.1466 -2.3223>, rot=<0.0465 -0.7289 -0.009 -0.6829>}
 upperarm_l {parent=clavicle_l, loc=<15.5111 0 0>, rot=<0.0226 -0.3234 -0.0088 -0.9459>}
 lowerarm_l {parent=upperarm_l, loc=<26.6514 0 0>, rot=<0.0391 -0.0099 -0.2453 -0.9686>}
 hand_l {parent=lowerarm_l, loc=<22.3491 0 0>, rot=<-0.7081 -0.0507 0.0975 -0.6975>}
 thumb_01_l {parent=hand_l, loc=<1.3118 -0.3191 -2.459>, rot=<0.6546 -0.4042 0.0664 -0.6354>}
 thumb_02_l {parent=thumb_01_l, loc=<5.9336 -0.0448 0.0401>, rot=<-0.0037 0.0213 0.298 -0.9543>}
 thumb_03_l {parent=thumb_02_l, loc=<2.2223 0.3682 -0.0907>, rot=<0.0118 -0.013 0.0066 -0.9998>}
 index_metacarpal_l {parent=hand_l, loc=<1.6535 0.0183 -2.96>, rot=<0.0158 -0.0124 -0.0173 -0.9996>}
 index_01_l {parent=index_metacarpal_l, loc=<8.3008 0.0357 -0.0524>, rot=<-0.0044 0.0485 0.0924 -0.9945>}
 index_02_l {parent=index_01_l, loc=<3.3587 -0.3004 0.0452>, rot=<-0.002 0.0027 0.3225 -0.9466>}
 index_03_l {parent=index_02_l, loc=<2.5578 -0.123 -0.0252>, rot=<-0.002 0.0067 0.2558 -0.9667>}
 middle_metacarpal_l {parent=hand_l, loc=<2.1856 -0.1725 -0.5984>, rot=<0.006 0.0168 -0.0329 -0.9993>}
 middle_01_l {parent=middle_metacarpal_l, loc=<7.3275 -0.2447 0.0129>, rot=<-0.0208 0.0042 0.0858 -0.9961>}
 middle_02_l {parent=middle_01_l, loc=<4.0123 0.1401 0.1175>, rot=<-0.0218 0.0258 0.3512 -0.9357>}
 middle_03_l {parent=middle_02_l, loc=<2.6077 0.0534 0.1373>, rot=<0.0082 0.0188 0.1708 -0.9851>}
 ring_metacarpal_l {parent=hand_l, loc=<2.2074 -0.3625 1.3638>, rot=<-0.0284 0.0722 -0.0458 -0.9959>}
 ring_01_l {parent=ring_metacarpal_l, loc=<7.0295 -0.6197 -0.1446>, rot=<-0.0489 -0.0252 0.1892 -0.9804>}
 ring_02_l {parent=ring_01_l, loc=<3.6174 0.4647 -0.1027>, rot=<0 0 0.2806 -0.9598>}
 ring_03_l {parent=ring_02_l, loc=<2.4152 0.114 0.0199>, rot=<0 0 0.1189 -0.9929>}
 pinky_metacarpal_l {parent=hand_l, loc=<1.7616 -0.9753 3.3814>, rot=<-0.0397 0.0814 -0.069 -0.9935>}
 pinky_01_l {parent=pinky_metacarpal_l, loc=<7.1709 -0.4831 -0.1094>, rot=<-0.0654 -0.0035 0.2242 -0.9723>}
 pinky_02_l {parent=pinky_01_l, loc=<2.8274 0.0893 -0.1043>, rot=<0.0037 0.0011 0.2854 -0.9584>}
 pinky_03_l {parent=pinky_02_l, loc=<2.2046 0.0307 -0.0078>, rot=<-0.0002 0.0098 0.1445 -0.9895>}
 deform_wrist_dn_base_l {parent=hand_l, loc=<0 -2.2342 0>, rot=<0 0 0 -1>}
 deform_wrist_dn_l {parent=deform_wrist_dn_base_l, loc=<0 0 0>, rot=<0 0 0 1>}
 deform_wrist_up_base_l {parent=hand_l, loc=<0 2.2682 0>, rot=<0 0 0 -1>}
 deform_wrist_up_l {parent=deform_wrist_up_base_l, loc=<0 0 0>, rot=<0 0 0 1>}
 lowerarm_twist_02_l {parent=lowerarm_l, loc=<9.6199 0 0>, rot=<0 0 0 -1>}
 lowerarm_twist_01_l {parent=lowerarm_l, loc=<17.8549 0 0>, rot=<0 0 0 -1>}
 dyn_wrist_chain_01 {parent=lowerarm_twist_01_l, loc=<-0.3192 -3.7773 2.5605>, rot=<-0.4008 -0.3453 0.3265 -0.7833>}
 dyn_wrist_chain_02 {parent=dyn_wrist_chain_01, loc=<2.0959 0 0>, rot=<-0.0426 0.061 -0.0305 -0.9968>}
 dyn_wrist_chain_03 {parent=dyn_wrist_chain_02, loc=<1.9945 0 0>, rot=<-0.1246 -0.0503 -0.0041 -0.9909>}
 dyn_wrist_chain_04 {parent=dyn_wrist_chain_03, loc=<2.4823 0 0>, rot=<-0.0032 0.0584 -0.0007 -0.9983>}
 dyn_wrist_chain_05 {parent=dyn_wrist_chain_04, loc=<1.9985 0 0>, rot=<-0.0028 -0.0745 -0.0135 -0.9971>}
 dyn_wrist_chain_06 {parent=dyn_wrist_chain_05, loc=<2.4582 0 0>, rot=<-0.013 0.066 0.0144 -0.9976>}
 dyn_wrist_chain_07 {parent=dyn_wrist_chain_06, loc=<1.9944 0 0>, rot=<-0.028 -0.0655 -0.0026 -0.9975>}
 dyn_wrist_chain_08 {parent=dyn_wrist_chain_07, loc=<2.4287 0 0>, rot=<-0.0163 0.0533 0.0236 -0.9982>}
 dyn_wrist_chain_09 {parent=dyn_wrist_chain_08, loc=<2.0015 0 0>, rot=<-0.0063 -0.0196 -0.0342 -0.9992>}
 dyn_wrist_chain_10 {parent=dyn_wrist_chain_09, loc=<2.4074 0 0>, rot=<-0.022 0.0052 0.0393 -0.999>}
 deform_elbow_in_base_l {parent=lowerarm_l, loc=<0.2112 5.259 0.8274>, rot=<0 0 0 -1>}
 deform_elbow_in_l {parent=deform_elbow_in_base_l, loc=<0 0 0>, rot=<0 0 0 1>}
 deform_elbow_out_base_l {parent=lowerarm_l, loc=<0 -4 0>, rot=<0 0 0 -1>}
 deform_elbow_out_l {parent=deform_elbow_out_base_l, loc=<0 0 0>, rot=<0 0 0 1>}
 upperarm_twist_01_l {parent=upperarm_l, loc=<0.0041 0 0>, rot=<0 0 0 1>}
 deform_upperarm_bk_base_l {parent=upperarm_twist_01_l, loc=<1.0813 -7.2193 -1.9453>, rot=<0 0 0 1>}
 deform_upperarm_bk_l {parent=deform_upperarm_bk_base_l, loc=<0 0 0>, rot=<0 0 0 1>}
 deform_upperarm_fr_base_l {parent=upperarm_twist_01_l, loc=<-0.3691 7.0506 -2.765>, rot=<0 0 0 1>}
 deform_upperarm_fr_l {parent=deform_upperarm_fr_base_l, loc=<0 0 0>, rot=<0 0 0 1>}
 deform_upperarm_up_base_l {parent=upperarm_twist_01_l, loc=<-3.2216 -1.1391 3.3311>, rot=<0 0 0 1>}
 deform_upperarm_up_l {parent=deform_upperarm_up_base_l, loc=<0 0 0>, rot=<0 0 0 1>}
 dyn_shoulder_tear_01_base_l {parent=upperarm_twist_01_l, loc=<-3.4364 3.551 6.9836>, rot=<0.0972 -0.2194 0.1654 -0.9566>}
 dyn_shoulder_tear_01_l {parent=dyn_shoulder_tear_01_base_l, loc=<0 0 0>, rot=<0 0 0 1>}
 dyn_shoulder_tear_02_base_l {parent=upperarm_twist_01_l, loc=<-4.0554 -4.9291 5.4729>, rot=<-0.0503 -0.3114 -0.2631 -0.9117>}
 dyn_shoulder_tear_02_l {parent=dyn_shoulder_tear_02_base_l, loc=<0 0 0>, rot=<0 0 0 1>}
 dyn_shoulder_tear_03_base_l {parent=upperarm_twist_01_l, loc=<-1.3 -8.0415 2.1709>, rot=<-0.1729 -0.3278 -0.5361 -0.7585>}
 dyn_shoulder_tear_03_l {parent=dyn_shoulder_tear_03_base_l, loc=<0 0 0>, rot=<0 0 0 1>}
 upperarm_twist_02_l {parent=upperarm_l, loc=<13.7941 0 0>, rot=<0 0 0 1>}
 deform_shoulder_armor_lower_l_base {parent=clavicle_l, loc=<9.6735 8.6597 4.8415>, rot=<0.4769 -0.0456 0.0407 -0.8768>}
 deform_shoulder_armor_lower_l {parent=deform_shoulder_armor_lower_l_base, loc=<0 0 0>, rot=<0 0 0 1>}
 deform_shoulder_armor_upper_l {parent=deform_shoulder_armor_lower_l, loc=<0.376 -7.8528 0.3738>, rot=<-0.1822 0.0164 -0.0109 -0.9831>}
 clavicle_r {parent=spine_05, loc=<4.3575 -0.1466 2.3223>, rot=<0.7289 0.0465 -0.6829 0.009>}
 upperarm_r {parent=clavicle_r, loc=<-15.5111 0 0>, rot=<0.0226 -0.3234 -0.0088 -0.9459>}
 lowerarm_r {parent=upperarm_r, loc=<-26.6514 0 0>, rot=<0.0391 -0.0099 -0.2453 -0.9686>}
 hand_r {parent=lowerarm_r, loc=<-22.3491 0 0>, rot=<-0.7081 -0.0507 0.0975 -0.6975>}
 thumb_01_r {parent=hand_r, loc=<-1.3118 0.3191 2.459>, rot=<0.6546 -0.4042 0.0664 -0.6354>}
 thumb_02_r {parent=thumb_01_r, loc=<-5.9336 0.0448 -0.0401>, rot=<-0.0037 0.0213 0.298 -0.9543>}
 thumb_03_r {parent=thumb_02_r, loc=<-2.2223 -0.3682 0.0907>, rot=<0.0118 -0.013 0.0066 -0.9998>}
 index_metacarpal_r {parent=hand_r, loc=<-1.6535 -0.0183 2.96>, rot=<0.0158 -0.0124 -0.0173 -0.9996>}
 index_01_r {parent=index_metacarpal_r, loc=<-8.3008 -0.0357 0.0524>, rot=<-0.0044 0.0485 0.0924 -0.9945>}
 index_02_r {parent=index_01_r, loc=<-3.3587 0.3004 -0.0452>, rot=<-0.002 0.0027 0.3225 -0.9466>}
 index_03_r {parent=index_02_r, loc=<-2.5578 0.123 0.0252>, rot=<-0.002 0.0067 0.2558 -0.9667>}
 middle_metacarpal_r {parent=hand_r, loc=<-2.1856 0.1725 0.5984>, rot=<0.006 0.0168 -0.0329 -0.9993>}
 middle_01_r {parent=middle_metacarpal_r, loc=<-7.3275 0.2447 -0.0129>, rot=<-0.0208 0.0042 0.0858 -0.9961>}
 middle_02_r {parent=middle_01_r, loc=<-4.0123 -0.1401 -0.1175>, rot=<-0.0218 0.0258 0.3512 -0.9357>}
 middle_03_r {parent=middle_02_r, loc=<-2.6077 -0.0534 -0.1373>, rot=<0.0082 0.0188 0.1708 -0.9851>}
 ring_metacarpal_r {parent=hand_r, loc=<-2.2074 0.3625 -1.3638>, rot=<-0.0284 0.0722 -0.0458 -0.9959>}
 ring_01_r {parent=ring_metacarpal_r, loc=<-7.0295 0.6197 0.1446>, rot=<-0.0489 -0.0252 0.1892 -0.9804>}
 ring_02_r {parent=ring_01_r, loc=<-3.6174 -0.4647 0.1027>, rot=<0 0 0.2806 -0.9598>}
 ring_03_r {parent=ring_02_r, loc=<-2.4152 -0.114 -0.0199>, rot=<0 0 0.1189 -0.9929>}
 pinky_metacarpal_r {parent=hand_r, loc=<-1.7616 0.9753 -3.3814>, rot=<-0.0397 0.0814 -0.069 -0.9935>}
 pinky_01_r {parent=pinky_metacarpal_r, loc=<-7.1709 0.4831 0.1094>, rot=<-0.0654 -0.0035 0.2242 -0.9723>}
 pinky_02_r {parent=pinky_01_r, loc=<-2.8274 -0.0893 0.1043>, rot=<0.0037 0.0011 0.2854 -0.9584>}
 pinky_03_r {parent=pinky_02_r, loc=<-2.2046 -0.0307 0.0078>, rot=<-0.0002 0.0098 0.1445 -0.9895>}
 deform_wrist_dn_base_r {parent=hand_r, loc=<-0.0004 2.2345 -0.0001>, rot=<0 0 0 -1>}
 deform_wrist_dn_r {parent=deform_wrist_dn_base_r, loc=<0 0 0>, rot=<0 0 0 1>}
 deform_wrist_up_base_r {parent=hand_r, loc=<0.0003 -2.2684 0.0001>, rot=<0 0 0 -1>}
 deform_wrist_up_r {parent=deform_wrist_up_base_r, loc=<0 0 0>, rot=<0 0 0 1>}
 lowerarm_twist_02_r {parent=lowerarm_r, loc=<-9.6199 0 0>, rot=<0 0 0 -1>}
 lowerarm_twist_01_r {parent=lowerarm_r, loc=<-17.8549 0 0>, rot=<0 0 0 -1>}
 deform_elbow_in_base_r {parent=lowerarm_r, loc=<-0.2111 -5.2591 -0.8275>, rot=<0 0 0 -1>}
 deform_elbow_in_r {parent=deform_elbow_in_base_r, loc=<-0.0001 0 0.0002>, rot=<0 0 0 1>}
 deform_elbow_out_base_r {parent=lowerarm_r, loc=<0 4 -0.0001>, rot=<0 0 0 -1>}
 deform_elbow_out_r {parent=deform_elbow_out_base_r, loc=<0 0 0>, rot=<0 0 0 1>}
 upperarm_twist_01_r {parent=upperarm_r, loc=<-0.0041 0 0>, rot=<0 0 0 -1>}
 deform_upperarm_bk_base_r {parent=upperarm_twist_01_r, loc=<-1.0814 7.2192 1.9455>, rot=<0 0 0 -1>}
 deform_upperarm_bk_r {parent=deform_upperarm_bk_base_r, loc=<0 0 0>, rot=<0 0 0 1>}
 deform_upperarm_fr_base_r {parent=upperarm_twist_01_r, loc=<0.3689 -7.0506 2.7652>, rot=<0 0 0 -1>}
 deform_upperarm_fr_r {parent=deform_upperarm_fr_base_r, loc=<0 0 0>, rot=<0 0 0 1>}
 deform_upperarm_up_base_r {parent=upperarm_twist_01_r, loc=<3.2214 1.1391 -3.3308>, rot=<0 0 0 -1>}
 deform_upperarm_up_r {parent=deform_upperarm_up_base_r, loc=<0 0 0>, rot=<0 0 0 -1>}
 dyn_shoulder_tear_05_base_r {parent=upperarm_twist_01_r, loc=<1.3002 8.0415 -2.1712>, rot=<-0.1729 -0.3278 -0.5361 -0.7585>}
 dyn_shoulder_tear_05_r {parent=dyn_shoulder_tear_05_base_r, loc=<0 0 0>, rot=<0 0 0 1>}
 dyn_shoulder_tear_04_base_r {parent=upperarm_twist_01_r, loc=<4.3659 4.457 -5.5358>, rot=<-0.0503 -0.3114 -0.2631 -0.9117>}
 dyn_shoulder_tear_04_r {parent=dyn_shoulder_tear_04_base_r, loc=<0 0 0>, rot=<0 0 0 1>}
 dyn_shoulder_tear_03_base_r {parent=upperarm_twist_01_r, loc=<5.4926 -0.2818 -5.9453>, rot=<0.0519 -0.3111 0.0424 -0.948>}
 dyn_shoulder_tear_03_r {parent=dyn_shoulder_tear_03_base_r, loc=<0 0 0>, rot=<0 0 0 1>}
 dyn_shoulder_tear_02_base_r {parent=upperarm_twist_01_r, loc=<3.8498 -3.5691 -6.6112>, rot=<0.0793 -0.3186 0.1747 -0.9283>}
 dyn_shoulder_tear_02_r {parent=dyn_shoulder_tear_02_base_r, loc=<0 0 0>, rot=<0 0 0 1>}
 dyn_shoulder_tear_01_base_r {parent=upperarm_twist_01_r, loc=<2.3702 -6.5955 -4.6656>, rot=<0.1549 -0.1649 0.6402 -0.7341>}
 dyn_shoulder_tear_01_r {parent=dyn_shoulder_tear_01_base_r, loc=<0 0 0>, rot=<0 0 0 1>}
 upperarm_twist_02_r {parent=upperarm_r, loc=<-13.7941 0 0>, rot=<0 0 0 -1>}
 deform_shoulder_armor_r_base {parent=clavicle_r, loc=<-9.6735 -8.6597 -4.8412>, rot=<0.4769 -0.0456 0.0407 -0.8768>}
 deform_shoulder_armor_lower_r {parent=deform_shoulder_armor_r_base, loc=<0 0 0>, rot=<0 0 0 1>}
 deform_shoulder_armor_upper_r {parent=deform_shoulder_armor_lower_r, loc=<-0.3759 7.8531 -0.374>, rot=<-0.1822 0.0164 -0.0109 -0.9831>}
 dyn_collar_r_01_base {parent=spine_05, loc=<13.8469 -8.3531 6.7785>, rot=<-0.163 0.2506 0.9149 0.2713>}
 dyn_collar_r_01 {parent=dyn_collar_r_01_base, loc=<0 0 0>, rot=<0 0 0 1>}
 dyn_collar_mid_01_base {parent=spine_05, loc=<11.3579 -11.0768 0>, rot=<0 0 0.1027 -0.9947>}
 dyn_collar_mid_01 {parent=dyn_collar_mid_01_base, loc=<0 0 0>, rot=<0 0 0 1>}
 dyn_collar_mid_02 {parent=dyn_collar_mid_01, loc=<3.5383 0 0>, rot=<0.0191 0.0372 -0.0259 -0.9988>}
 dyn_collar_l_01_base {parent=spine_05, loc=<13.8468 -8.353 -6.7785>, rot=<-0.2506 -0.163 0.2713 -0.9149>}
 dyn_collar_l_01 {parent=dyn_collar_l_01_base, loc=<0 0 0>, rot=<0 0 0 1>}
 deform_pec_aim_base_l {parent=spine_04, loc=<7.8732 12.6388 -2.4123>, rot=<0.0102 -0.4602 0.3979 -0.7936>}
 deform_pec_aim_l {parent=deform_pec_aim_base_l, loc=<0 0 0>, rot=<0 0 0 1>}
 deform_pec_base_l {parent=deform_pec_aim_l, loc=<10.766 0 0>, rot=<0 0 0 1>}
 deform_pec_l {parent=deform_pec_base_l, loc=<0.055 0 0>, rot=<0 0 0 1>}
 deform_pec_aim_base_r {parent=spine_04, loc=<7.8727 12.6388 2.4123>, rot=<-0.4602 -0.0102 0.7936 0.3979>}
 deform_pec_aim_r {parent=deform_pec_aim_base_r, loc=<0 0 0>, rot=<0 0 0 1>}
 deform_pec_base_r {parent=deform_pec_aim_r, loc=<-10.7662 -0.0006 -0.0007>, rot=<0 0 0 1>}
 deform_pec_r {parent=deform_pec_base_r, loc=<0 0 0>, rot=<0 0 0 1>}
 dyn_jacket_front_l_01_base {parent=spine_04, loc=<3.1008 8.5647 -14.7764>, rot=<-0.0835 0.4801 0.8656 -0.1153>}
 dyn_jacket_front_l_01 {parent=dyn_jacket_front_l_01_base, loc=<0 0 0>, rot=<0 0 0 1>}
 dyn_jacket_front_l_02 {parent=dyn_jacket_front_l_01, loc=<8.4908 0 0>, rot=<0.0181 0.0312 -0.0184 -0.9992>}
 dyn_jacket_front_l_03 {parent=dyn_jacket_front_l_02, loc=<7.4252 0 0>, rot=<0 0 0 1>}
 dyn_jacket_side_l_01_base {parent=spine_04, loc=<1.5426 -1.8498 -17.7522>, rot=<0.0774 0.6917 0.7134 -0.082>}
 dyn_jacket_side_l_01 {parent=dyn_jacket_side_l_01_base, loc=<0 0 0>, rot=<0 0 0 1>}
 dyn_jacket_side_l_02 {parent=dyn_jacket_side_l_01, loc=<6.633 0 0>, rot=<0.0191 0.0372 -0.0259 -0.9988>}
 dyn_jacket_rear_l_01_base {parent=spine_04, loc=<6.019 -11.8387 -12.1523>, rot=<-0.1235 -0.987 -0.0986 0.028>}
 dyn_jacket_rear_l_01 {parent=dyn_jacket_rear_l_01_base, loc=<0 0 0>, rot=<0 0 0 1>}
 dyn_jacket_rear_l_02 {parent=dyn_jacket_rear_l_01, loc=<5.6507 0 0>, rot=<0.0191 0.0372 -0.0259 -0.9988>}
 dyn_jacket_l_rip_01_base {parent=dyn_jacket_rear_l_02, loc=<5.1978 -0.9117 -1.4876>, rot=<0.1914 -0.0454 0.2469 -0.9489>}
 dyn_jacket_l_rip_01 {parent=dyn_jacket_l_rip_01_base, loc=<0 0 0>, rot=<0 0 0 1>}
 dyn_jacket_rear_01_base {parent=spine_04, loc=<6.4669 -13.0697 0>, rot=<-0.0705 -0.9975 0 0>}
 dyn_jacket_rear_01 {parent=dyn_jacket_rear_01_base, loc=<0 0 0>, rot=<0 0 0 1>}
 dyn_jacket_rear_02 {parent=dyn_jacket_rear_01, loc=<6.0304 0 0>, rot=<0.0191 0.0372 -0.0259 -0.9988>}
 dyn_jacket_rear_rip_01_base {parent=dyn_jacket_rear_02, loc=<1.8771 -0.0936 -0.1283>, rot=<-0.0457 -0.0298 0.1249 -0.9907>}
 dyn_jacket_rear_rip_01 {parent=dyn_jacket_rear_rip_01_base, loc=<0 0 0>, rot=<0 0 0 1>}
 dyn_jacket_l_rip_02_base {parent=dyn_jacket_rear_02, loc=<4.5032 -0.9755 5.836>, rot=<0.0828 -0.0588 0.2256 -0.9689>}
 dyn_jacket_l_rip_02 {parent=dyn_jacket_l_rip_02_base, loc=<0 0 0>, rot=<0 0 0 1>}
 dyn_jacket_r_rip_01_base {parent=dyn_jacket_rear_02, loc=<2.9205 -0.4382 -4.402>, rot=<-0.055 -0.0502 0.9928 0.0934>}
 dyn_jacket_r_rip_01 {parent=dyn_jacket_r_rip_01_base, loc=<0 0 0>, rot=<0 0 0 1>}
 dyn_jacket_side_r_01_base {parent=spine_04, loc=<1.543 -1.8498 17.7522>, rot=<-0.6917 0.0774 -0.082 -0.7134>}
 dyn_jacket_side_r_01 {parent=dyn_jacket_side_r_01_base, loc=<0 0 0>, rot=<0 0 0 1>}
 dyn_jacket_side_r_02 {parent=dyn_jacket_side_r_01, loc=<-6.6335 0.0001 0>, rot=<0.0191 0.0372 -0.0259 -0.9988>}
 dyn_jacket_rear_r_01_base {parent=spine_04, loc=<6.0187 -11.8387 12.1523>, rot=<0.987 -0.1235 0.028 0.0986>}
 dyn_jacket_rear_r_01 {parent=dyn_jacket_rear_r_01_base, loc=<0 0 0>, rot=<0 0 0 -1>}
 dyn_jacket_rear_r_02 {parent=dyn_jacket_rear_r_01, loc=<-5.65 -0.0003 -0.0001>, rot=<0.0191 0.0372 -0.0259 -0.9988>}
 dyn_jacket_r_rip_03_base {parent=dyn_jacket_rear_r_02, loc=<-4.2553 0.9414 1.6467>, rot=<0.1753 -0.0713 0.2205 -0.9568>}
 dyn_jacket_r_rip_03 {parent=dyn_jacket_r_rip_03_base, loc=<0 0 0>, rot=<0 0 0 1>}
 dyn_jacket_r_rip_02_base {parent=dyn_jacket_rear_r_02, loc=<-2.7567 1.0632 5.6693>, rot=<0 0 0.1138 -0.9935>}
 dyn_jacket_r_rip_02 {parent=dyn_jacket_r_rip_02_base, loc=<0 0 0>, rot=<0 0 0 1>}
 dyn_jacket_front_r_01_base {parent=spine_04, loc=<3.1008 8.5647 14.7764>, rot=<-0.4801 -0.0835 -0.1153 -0.8656>}
 dyn_jacket_front_r_01 {parent=dyn_jacket_front_r_01_base, loc=<0 0 0>, rot=<0 0 0 -1>}
 dyn_jacket_front_r_02 {parent=dyn_jacket_front_r_01, loc=<-8.491 0 0>, rot=<0.0181 0.0312 -0.0184 -0.9992>}
 dyn_jacket_front_r_03 {parent=dyn_jacket_front_r_02, loc=<-7.4251 0 0>, rot=<0 0 0 -1>}
 thigh_l {parent=pelvis, loc=<-2.5693 -0.0109 -7.7501>, rot=<0.0012 0.0572 0.0155 -0.9982>}
 calf_l {parent=thigh_l, loc=<-39.0484 0 0>, rot=<0.0006 -0.0161 -0.0391 -0.9991>}
 foot_l {parent=calf_l, loc=<-42.561 0.1077 0.4292>, rot=<0.0418 -0.0455 0.0343 -0.9975>}
 ball_l {parent=foot_l, loc=<-6.1599 13.8264 0.6806>, rot=<0.0001 -0.0001 -0.7063 -0.7079>}
 calf_twist_02_l {parent=calf_l, loc=<-18.8064 -0.0967 0.0397>, rot=<0.0237 0.0095 -0.0035 -0.9997>}
 calf_twist_01_l {parent=calf_l, loc=<-35.813 -0.0967 0.0397>, rot=<0.0237 0.0095 -0.0035 -0.9997>}
 deform_knee_in_base_l {parent=calf_l, loc=<1.298 -5.2957 -0.0496>, rot=<0 0 0 -1>}
 deform_knee_in_l {parent=deform_knee_in_base_l, loc=<0 0 0>, rot=<0 0 0 1>}
 deform_knee_out_base_l {parent=calf_l, loc=<0 4 0>, rot=<0 0 0 -1>}
 deform_knee_out_l {parent=deform_knee_out_base_l, loc=<0 0 0>, rot=<0 0 0 1>}
 thigh_twist_01_l {parent=thigh_l, loc=<-21.6303 0 0>, rot=<0.0013 0 0 -1>}
 deform_groin_base_l {parent=thigh_l, loc=<-2.1908 8.1142 -1.7442>, rot=<0 0 0 -1>}
 deform_groin_l {parent=deform_groin_base_l, loc=<0 0 0>, rot=<0 0 0 1>}
 thigh_r {parent=pelvis, loc=<-2.5693 -0.0109 7.7501>, rot=<0.0572 -0.0012 0.9982 0.0155>}
 calf_r {parent=thigh_r, loc=<39.0484 0 0>, rot=<0.0006 -0.0161 -0.0391 -0.9991>}
 foot_r {parent=calf_r, loc=<42.561 -0.1077 -0.4292>, rot=<0.0418 -0.0455 0.0343 -0.9975>}
 ball_r {parent=foot_r, loc=<6.1599 -13.8264 -0.6806>, rot=<0.0001 -0.0001 -0.7063 -0.7079>}
 calf_twist_02_r {parent=calf_r, loc=<18.8064 0.0967 -0.0397>, rot=<0.0237 0.0095 -0.0035 -0.9997>}
 calf_twist_01_r {parent=calf_r, loc=<35.813 0.0967 -0.0397>, rot=<0.0237 0.0095 -0.0035 -0.9997>}
 deform_knee_in_base_r {parent=calf_r, loc=<-1.2958 5.2877 0.0529>, rot=<0 0 0 -1>}
 deform_knee_in_r {parent=deform_knee_in_base_r, loc=<0 0 0>, rot=<0 0 0 1>}
 deform_knee_out_base_r {parent=calf_r, loc=<0 -3.9978 -0.0009>, rot=<0 0 0 -1>}
 deform_knee_out_r {parent=deform_knee_out_base_r, loc=<0 0 0>, rot=<0 0 0 1>}
 thigh_twist_01_r {parent=thigh_r, loc=<21.6303 0 0>, rot=<0.0013 0 0 -1>}
 deform_groin_base_r {parent=thigh_r, loc=<2.1908 -8.1142 1.7442>, rot=<0 0 0 -1>}
 deform_groin_r {parent=deform_groin_base_r, loc=<0 0 0>, rot=<0 0 0 1>}
 deform_glute_base_l {parent=pelvis, loc=<-8.8064 -8.8222 -8.4287>, rot=<0.0275 0.0501 -0.4472 -0.8926>}
 deform_glute_l {parent=deform_glute_base_l, loc=<0 0 0>, rot=<0 0 0 1>}
 deform_glute_base_r {parent=pelvis, loc=<-8.8062 -8.8222 8.4287>, rot=<0.0501 -0.0275 0.8926 -0.4472>}
 deform_glute_r {parent=deform_glute_base_r, loc=<0 0 0>, rot=<0 0 0 1>}
 TailPoint {parent=pelvis, loc=<7.5462 -9.758 0>, rot=<0 0 0 1>}
 deform_belt_buckle {parent=pelvis, loc=<6.0417 10.1265 0>, rot=<0 0 0 1>}
 ik_hand_root {parent=root, loc=<0 0 0>, rot=<0 0 0 1>}
 ik_foot_root {parent=root, loc=<0 0 0>, rot=<0 0 0 1>}
 ik_foot_l {parent=ik_foot_root, loc=<15.2608 -4.4176 9.4432>, rot=<0.0357 0.7045 0.0284 -0.7082>}
 ik_foot_r {parent=ik_foot_root, loc=<-15.2608 -4.4176 9.4432>, rot=<0.7082 0.0284 -0.7045 0.0357>}
]
variables [IsSkydiveFloating {type=bool, default=False}]
