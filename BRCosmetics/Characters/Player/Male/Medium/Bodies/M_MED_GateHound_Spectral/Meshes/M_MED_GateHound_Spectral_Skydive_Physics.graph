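# Physics asset for the skydive mesh of the GateHound Spectral male medium body, covering its dynamic jacket bones.
# Seven capsule chains (front, side, rear left/right, rear center) start at tiny base bodies jointed to spine_04 with free rotation.
# Angular limits constrain each segment, tightest at rear center; rear left and right limits differ.
# Three large capsules on spine_03 through spine_05 form the torso.
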PhysicsAsset /BRCosmetics/Characters/Player/Male/Medium/Bodies/M_MED_GateHound_Spectral/Meshes/M_MED_GateHound_Spectral_Skydive_Physics
bodies {
 dyn_jacket_front_l_01 [{
   shapeType=Capsule
   length=6.0966
   radius=2.5547
   shapeName=dyn_jacket_front_l_01_capsule
   center=<1.9028 -0.5547 -0.2232>
   rotation=<rot 21.8149 94.1657 -176.0512>
  }]
 dyn_jacket_front_l_01_base [{shapeType=Capsule, length=0.505, radius=0.505, shapeName=dyn_jacket_front_l_01_base_capsule, rotation=<rot 0 -90 0>}]
 dyn_jacket_front_l_02 [{
   shapeType=Capsule
   length=4.6107
   radius=2.8715
   shapeName=dyn_jacket_front_l_02_capsule
   center=<2.1475 -0.038 -0.283>
   rotation=<rot -21.8731 -90.2924 -178.8065>
  }]
 dyn_jacket_front_l_03 [{
   shapeType=Capsule
   length=6.9838
   radius=2.9686
   shapeName=dyn_jacket_front_l_03_capsule
   center=<2.5778 0.3512 -0.1253>
   rotation=<rot 19.464 73.7401 -93.7124>
  }]
 dyn_jacket_front_r_01 [{
   shapeType=Capsule
   length=6.0966
   radius=2.5547
   shapeName=dyn_jacket_front_l_01_capsule
   center=<-1.9028 0.5547 0.2232>
   rotation=<rot 21.8149 94.1657 3.9488>
  }]
 dyn_jacket_front_r_01_base [{shapeType=Capsule, length=0.505, radius=0.505, shapeName=dyn_jacket_front_r_01_base_capsule, rotation=<rot 0 -90 0>}]
 dyn_jacket_front_r_02 [{
   shapeType=Capsule
   length=4.6107
   radius=2.8715
   shapeName=dyn_jacket_front_l_02_capsule
   center=<-2.1475 0.038 0.283>
   rotation=<rot -21.8731 -90.2924 1.1935>
  }]
 dyn_jacket_front_r_03 [{
   shapeType=Capsule
   length=6.9838
   radius=2.9686
   shapeName=dyn_jacket_front_l_03_capsule
   center=<-2.5778 -0.3512 0.1253>
   rotation=<rot 19.464 73.7401 86.2876>
  }]
 dyn_jacket_rear_01 [{
   shapeType=Capsule
   length=10.375
   radius=1.9425
   shapeName=dyn_jacket_rear_01_capsule
   center=<1.6577 -2.0935 0>
   rotation=<rot 0 -107.7237 0>
  }]
 dyn_jacket_rear_01_base [{shapeType=Capsule, length=0.505, radius=0.505, shapeName=dyn_jacket_rear_01_base_capsule, rotation=<rot 0 -90 0>}]
 dyn_jacket_rear_02 [{
   shapeType=Capsule
   length=8.6602
   radius=3.6875
   shapeName=dyn_jacket_rear_02_capsule
   center=<2.9999 -3.3058 -0.1398>
   rotation=<rot 1.1603 -105.304 4.232>
  }]
 dyn_jacket_rear_l_01 [{
   shapeType=Capsule
   length=8.8652
   radius=2.862
   shapeName=dyn_jacket_rear_l_01_capsule
   center=<2.94 -0.9129 0.5732>
   rotation=<rot 24.3929 -98.3783 2.6214>
  }]
 dyn_jacket_rear_l_01_base [{shapeType=Capsule, length=0.505, radius=0.505, shapeName=dyn_jacket_rear_l_01_base_capsule, rotation=<rot 0 -90 0>}]
 dyn_jacket_rear_l_02 [{
   shapeType=Capsule
   length=6.984
   radius=4.2037
   shapeName=dyn_jacket_rear_l_02_capsule
   center=<3.4198 -2.4045 -0.2362>
   rotation=<rot -27.2271 73.0861 -7.3497>
  }]
 dyn_jacket_rear_r_01 [{
   shapeType=Capsule
   length=9.5527
   radius=2.1745
   shapeName=dyn_jacket_rear_l_01_capsule
   center=<-2.8683 1.4 -0.7964>
   rotation=<rot 24.3929 -98.3783 179.123>
  }]
 dyn_jacket_rear_r_01_base [{shapeType=Capsule, length=0.505, radius=0.505, shapeName=dyn_jacket_rear_r_01_base_capsule, rotation=<rot 0 -90 0>}]
 dyn_jacket_rear_r_02 [{
   shapeType=Capsule
   length=6.984
   radius=4.2037
   shapeName=dyn_jacket_rear_l_02_capsule
   center=<-3.4198 2.4045 0.2362>
   rotation=<rot -27.2271 73.0861 172.6503>
  }]
 dyn_jacket_side_l_01 [{
   shapeType=Capsule
   length=4.4446
   radius=1.8364
   shapeName=dyn_jacket_side_l_01_capsule
   center=<1.6733 -0.2711 -0.4>
   rotation=<rot -3.8724 -90 8.8088>
  }]
 dyn_jacket_side_l_01_base [{shapeType=Capsule, length=0.505, radius=0.505, shapeName=dyn_jacket_side_l_01_base_capsule, rotation=<rot 0 -90 0>}]
 dyn_jacket_side_l_02 [{
   shapeType=Capsule
   length=7.3506
   radius=2.2657
   shapeName=dyn_jacket_side_l_02_capsule
   center=<2.9428 -1.0994 -0.6004>
   rotation=<rot 0 82.3502 -104.9387>
  }]
 dyn_jacket_side_r_01 [{
   shapeType=Capsule
   length=4.4446
   radius=1.8364
   shapeName=dyn_jacket_side_l_01_capsule
   center=<-1.6733 0.2711 0.4>
   rotation=<rot -3.8724 -90 -171.1912>
  }]
 dyn_jacket_side_r_01_base [{shapeType=Capsule, length=0.505, radius=0.505, shapeName=dyn_jacket_side_r_01_base_capsule, rotation=<rot 0 -90 0>}]
 dyn_jacket_side_r_02 [{
   shapeType=Capsule
   length=7.3506
   radius=2.2657
   shapeName=dyn_jacket_side_l_02_capsule
   center=<-1.3021 0.0224 0.1283>
   rotation=<rot 0 82.3502 97.9018>
  }]
 spine_03 [{shapeType=Capsule, length=4.2888, radius=12.5087, center=<1.259 1.0775 -0.0686>, rotation=<rot -0.1354 -82.147 90.9813>}]
 spine_04 [{shapeType=Capsule, length=4.3263, radius=13.67, center=<0.434 0.9488 -0.0024>, rotation=<rot -0.2326 -90.1047 90.716>}]
 spine_05 [{shapeType=Capsule, length=13.8931, radius=12.7852, center=<11.4014 -1.8234 -0.0375>, rotation=<rot 0.5357 -90 -0.2787>}]
}
constraints [
 {bone1Name=spine_04, bone2Name=spine_03, swing1LimitDegrees=45, swing2LimitDegrees=45, twistLimitDegrees=45}
 {bone1Name=spine_05, bone2Name=spine_04, swing1LimitDegrees=45, swing2LimitDegrees=45, twistLimitDegrees=45}
 {
  bone1Name=dyn_jacket_front_l_01_base
  bone2Name=spine_04
  swing1LimitDegrees=45
  swing1Motion=Free
  swing2LimitDegrees=45
  swing2Motion=Free
  twistLimitDegrees=45
  twistMotion=Free
 }
 {bone1Name=dyn_jacket_front_l_01, bone2Name=dyn_jacket_front_l_01_base, swing1LimitDegrees=32, swing2LimitDegrees=12, twistLimitDegrees=12}
 {bone1Name=dyn_jacket_front_l_02, bone2Name=dyn_jacket_front_l_01, swing1LimitDegrees=32, swing2LimitDegrees=12, twistLimitDegrees=12}
 {bone1Name=dyn_jacket_front_l_03, bone2Name=dyn_jacket_front_l_02, swing1LimitDegrees=32, swing2LimitDegrees=32, twistLimitDegrees=32}
 {
  bone1Name=dyn_jacket_front_r_01_base
  bone2Name=spine_04
  swing1LimitDegrees=45
  swing1Motion=Free
  swing2LimitDegrees=45
  swing2Motion=Free
  twistLimitDegrees=45
  twistMotion=Free
 }
 {
  bone1Name=dyn_jacket_rear_01_base
  bone2Name=spine_04
  swing1LimitDegrees=45
  swing1Motion=Free
  swing2LimitDegrees=45
  swing2Motion=Free
  twistLimitDegrees=45
  twistMotion=Free
 }
 {bone1Name=dyn_jacket_rear_01, bone2Name=dyn_jacket_rear_01_base, swing1LimitDegrees=8, swing2LimitDegrees=8, twistLimitDegrees=8}
 {bone1Name=dyn_jacket_rear_02, bone2Name=dyn_jacket_rear_01, swing1LimitDegrees=8, swing2LimitDegrees=8, twistLimitDegrees=8}
 {
  bone1Name=dyn_jacket_rear_l_01_base
  bone2Name=spine_04
  swing1LimitDegrees=45
  swing1Motion=Free
  swing2LimitDegrees=45
  swing2Motion=Free
  twistLimitDegrees=45
  twistMotion=Free
 }
 {
  bone1Name=dyn_jacket_rear_r_01_base
  bone2Name=spine_04
  swing1LimitDegrees=45
  swing1Motion=Free
  swing2LimitDegrees=45
  swing2Motion=Free
  twistLimitDegrees=45
  twistMotion=Free
 }
 {
  bone1Name=dyn_jacket_side_l_01_base
  bone2Name=spine_04
  swing1LimitDegrees=45
  swing1Motion=Free
  swing2LimitDegrees=45
  swing2Motion=Free
  twistLimitDegrees=45
  twistMotion=Free
 }
 {bone1Name=dyn_jacket_side_l_01, bone2Name=dyn_jacket_side_l_01_base, swing1LimitDegrees=18, swing2LimitDegrees=8, twistLimitDegrees=8}
 {bone1Name=dyn_jacket_side_l_02, bone2Name=dyn_jacket_side_l_01, swing1LimitDegrees=18, swing2LimitDegrees=8, twistLimitDegrees=8}
 {
  bone1Name=dyn_jacket_side_r_01_base
  bone2Name=spine_04
  swing1LimitDegrees=45
  swing1Motion=Free
  swing2LimitDegrees=45
  swing2Motion=Free
  twistLimitDegrees=45
  twistMotion=Free
 }
 {bone1Name=dyn_jacket_side_r_01, bone2Name=dyn_jacket_side_r_01_base, swing1LimitDegrees=18, swing2LimitDegrees=8, twistLimitDegrees=8}
 {bone1Name=dyn_jacket_side_r_02, bone2Name=dyn_jacket_side_r_01, swing1LimitDegrees=18, swing2LimitDegrees=8, twistLimitDegrees=8}
 {bone1Name=dyn_jacket_front_r_01, bone2Name=dyn_jacket_front_r_01_base, swing1LimitDegrees=32, swing2LimitDegrees=12, twistLimitDegrees=12}
 {bone1Name=dyn_jacket_rear_r_01, bone2Name=dyn_jacket_rear_r_01_base, swing1LimitDegrees=18, swing2LimitDegrees=8, twistLimitDegrees=8}
 {bone1Name=dyn_jacket_front_r_03, bone2Name=dyn_jacket_front_r_02, swing1LimitDegrees=32, swing2LimitDegrees=32, twistLimitDegrees=32}
 {bone1Name=dyn_jacket_front_r_02, bone2Name=dyn_jacket_front_r_01, swing1LimitDegrees=32, swing2LimitDegrees=12, twistLimitDegrees=12}
 {bone1Name=dyn_jacket_rear_r_02, bone2Name=dyn_jacket_rear_r_01, swing1LimitDegrees=18, swing2LimitDegrees=8, twistLimitDegrees=8}
 {bone1Name=dyn_jacket_rear_l_01, bone2Name=dyn_jacket_rear_l_01_base, swing1LimitDegrees=32, swing2LimitDegrees=32, twistLimitDegrees=32}
 {bone1Name=dyn_jacket_rear_l_02, bone2Name=dyn_jacket_rear_l_01, swing1LimitDegrees=10, swing2LimitDegrees=8, twistLimitDegrees=8}
]
