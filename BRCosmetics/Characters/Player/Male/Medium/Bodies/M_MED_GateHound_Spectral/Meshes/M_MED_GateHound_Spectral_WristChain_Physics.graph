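# Physics asset for the left wrist chain on the Spectral GateHound medium male player cosmetic body.
# Ten dyn_wrist_chain capsule bodies form a serial chain hanging from lowerarm_twist_01_l, with a looser swing limit at the root and tight, uniform limits between links.
# Links mostly alternate large and small capsules.
# Larger hand_l and lowerarm_twist_01_l capsules cover the arm, and the last link has no further constraint.
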PhysicsAsset /BRCosmetics/Characters/Player/Male/Medium/Bodies/M_MED_GateHound_Spectral/Meshes/M_MED_GateHound_Spectral_WristChain_Physics
bodies {
 dyn_wrist_chain_01 [{
   shapeType=Capsule
   length=1.4375
   radius=2.005
   shapeName=dyn_wrist_chain_01_capsule
   center=<0.8721 0.0224 0.0547>
   rotation=<rot 0 88.8389 -83.1236>
  }]
 dyn_wrist_chain_02 [{
   shapeType=Capsule
   length=2.2703
   radius=1.9077
   shapeName=dyn_wrist_chain_02_capsule
   center=<1.0045 -0.0011 0.0355>
   rotation=<rot 0 -94.9286 -87.3973>
  }]
 dyn_wrist_chain_03 [{
   shapeType=Capsule
   length=0.505
   radius=0.505
   shapeName=dyn_wrist_chain_03_capsule
   center=<1.2466 -0.0055 0.0308>
   rotation=<rot 0 91.3584 -89.1683>
  }]
 dyn_wrist_chain_04 [{
   shapeType=Capsule
   length=2.2703
   radius=1.9671
   shapeName=dyn_wrist_chain_04_capsule
   center=<1.0044 0.0016 0.0038>
   rotation=<rot 0 -95.3654 -88.2223>
  }]
 dyn_wrist_chain_05 [{
   shapeType=Capsule
   length=0.505
   radius=0.505
   shapeName=dyn_wrist_chain_05_capsule
   center=<1.2027 0.0157 -0.0146>
   rotation=<rot 0 89.5887 -86.2033>
  }]
 dyn_wrist_chain_06 [{
   shapeType=Capsule
   length=2.2703
   radius=1.9643
   shapeName=dyn_wrist_chain_06_capsule
   center=<1.0032 0.0014 0.0035>
   rotation=<rot 0 -95.3376 -88.3097>
  }]
 dyn_wrist_chain_07 [{
   shapeType=Capsule
   length=0.505
   radius=0.505
   shapeName=dyn_wrist_chain_07_capsule
   center=<1.2566 -0.0055 0.0015>
   rotation=<rot 0 89.8886 -86.6885>
  }]
 dyn_wrist_chain_08 [{
   shapeType=Capsule
   length=2.2703
   radius=1.9697
   shapeName=dyn_wrist_chain_08_capsule
   center=<1.0074 0.0024 0.0112>
   rotation=<rot 0 -94.8534 -89.0412>
  }]
 dyn_wrist_chain_09 [{
   shapeType=Capsule
   length=0.505
   radius=0.505
   shapeName=dyn_wrist_chain_09_capsule
   center=<1.2455 -0.0308 0.0042>
   rotation=<rot 0 87.3077 -88.829>
  }]
 dyn_wrist_chain_10 [{
   shapeType=Capsule
   length=2.2703
   radius=1.9656
   shapeName=dyn_wrist_chain_10_capsule
   center=<1.0067 0.0032 0.0145>
   rotation=<rot 0 -94.7462 -89.735>
  }]
 hand_l [{shapeType=Capsule, length=6.625, radius=5, shapeName=hand_l_capsule, center=<4.9648 0.2276 0.3337>, rotation=<rot -9.6849 -89.9999 -93.9055>}]
 lowerarm_twist_01_l [{
   shapeType=Capsule
   length=9.4097
   radius=4.1404
   shapeName=lowerarm_twist_01_l_capsule
   center=<-2.4278 -0.7073 1.0556>
   rotation=<rot 1.8742 -88.3017 -77.8634>
  }]
}
constraints [
 {bone1Name=dyn_wrist_chain_01, bone2Name=lowerarm_twist_01_l, swing1LimitDegrees=32, swing2LimitDegrees=32, twistLimitDegrees=14}
 {bone1Name=dyn_wrist_chain_02, bone2Name=dyn_wrist_chain_01, swing1LimitDegrees=14, swing2LimitDegrees=14, twistLimitDegrees=14}
 {bone1Name=dyn_wrist_chain_03, bone2Name=dyn_wrist_chain_02, swing1LimitDegrees=14, swing2LimitDegrees=14, twistLimitDegrees=14}
 {bone1Name=dyn_wrist_chain_04, bone2Name=dyn_wrist_chain_03, swing1LimitDegrees=14, swing2LimitDegrees=14, twistLimitDegrees=14}
 {bone1Name=dyn_wrist_chain_05, bone2Name=dyn_wrist_chain_04, swing1LimitDegrees=14, swing2LimitDegrees=14, twistLimitDegrees=14}
 {bone1Name=dyn_wrist_chain_06, bone2Name=dyn_wrist_chain_05, swing1LimitDegrees=14, swing2LimitDegrees=14, twistLimitDegrees=14}
 {bone1Name=dyn_wrist_chain_07, bone2Name=dyn_wrist_chain_06, swing1LimitDegrees=14, swing2LimitDegrees=14, twistLimitDegrees=14}
 {bone1Name=dyn_wrist_chain_08, bone2Name=dyn_wrist_chain_07, swing1LimitDegrees=14, swing2LimitDegrees=14, twistLimitDegrees=14}
 {bone1Name=dyn_wrist_chain_09, bone2Name=dyn_wrist_chain_08, swing1LimitDegrees=14, swing2LimitDegrees=14, twistLimitDegrees=14}
 {bone1Name=dyn_wrist_chain_10, bone2Name=dyn_wrist_chain_09, swing1LimitDegrees=14, swing2LimitDegrees=14, twistLimitDegrees=14}
]
